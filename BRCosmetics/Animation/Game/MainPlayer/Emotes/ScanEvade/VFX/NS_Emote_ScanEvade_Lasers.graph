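# Niagara system for the ScanEvade player emote that shows seven red laser beams, spark motes and a pulse wave.
# Laser bursts appear staggered on indexed skeletal mesh bones or sockets, selected by spawn group, then shrink and fade together on System.Age curves.
# Motes burst along each beam using sampled bone position and rotation, streaking by speed.
# Wave fires late, layering expanding mesh and sprite effects with separate age curves and platform-dependent material parameters.
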NiagaraSystem /BRCosmetics/Animation/Game/MainPlayer/Emotes/ScanEvade/VFX/NS_Emote_ScanEvade_Lasers
dependencies {
 usedDataInterfaces [
  &NiagaraDataInterfaceVectorCurve
  &NiagaraDataInterfaceArrayInt32
  &NiagaraDataInterfaceCurve
  &NiagaraDataInterfaceColorCurve
  &NiagaraDataInterfaceArrayFloat
  &NiagaraDataInterfacePlatformSet
  &NiagaraDataInterfaceSkeletalMesh
 ]
 usedDynamicInputs [&FNTimeOfDayScalar]
 usedModules [
  &self:SystemSpawnScript.SystemScriptSource.SystemScriptGraph.NiagaraNodeAssignment_0.SetVariables_#1
  &self:SystemSpawnScript.SystemScriptSource.SystemScriptGraph.NiagaraNodeAssignment_1.SetVariables_#2
  &SystemState
 ]
}
emitters {
 Laser {
  emitterUpdateScript [
   &EmitterState()
   &SpawnBurst_Instantaneous(Spawn Time=0.1, Use Spawn Probability=false, UseLoopCountLimit=false)
   &SpawnBurst_Instantaneous(Spawn Group=1, Spawn Time=0.2, Use Spawn Probability=false, UseLoopCountLimit=false) {module=SpawnBurst_Instantaneous002}
   &SpawnBurst_Instantaneous(Spawn Group=2, Spawn Time=0.4, Use Spawn Probability=false, UseLoopCountLimit=false) {module=SpawnBurst_Instantaneous003}
   &SpawnBurst_Instantaneous(Spawn Group=3, Spawn Time=0.6, Use Spawn Probability=false, UseLoopCountLimit=false) {module=SpawnBurst_Instantaneous004}
   &SpawnBurst_Instantaneous(Spawn Group=4, Spawn Time=0.4, Use Spawn Probability=false, UseLoopCountLimit=false) {module=SpawnBurst_Instantaneous005}
   &SpawnBurst_Instantaneous(Spawn Group=5, Spawn Time=0.6, Use Spawn Probability=false, UseLoopCountLimit=false) {module=SpawnBurst_Instantaneous006}
   &SpawnBurst_Instantaneous(Spawn Group=6, Spawn Time=0.8, Use Spawn Probability=false, UseLoopCountLimit=false) {module=SpawnBurst_Instantaneous007}
  ]
  particleSpawnScript [
   &InitializeParticle(
    Color=<rgba 0.6458 0.0159 0.0135 0.6>
    Lifetime=&MergeHelper_FloatSwitch(0=1.67, 1=1.47, 2=1.27, 3=1.07, 4=1.27, 5=1.07, 6=0.87, Index=&SpawnGroup(), Num=::NewEnumerator5)
    Mesh Scale=&MergeHelper_VectorSwitch(
     0=<0.1 0.46 0.1>
     1=<0.1 0.45 0.1>
     2=<0.1 0.5 0.1>
     3=<0.1 0.5 0.1>
     4=<0.1 0.4 0.1>
     5=<0.1 0.42 0.1>
     6=<0.1 0.45 0.1>
     Index=&SpawnGroup()
     Num=::NewEnumerator5
    )
    Mesh Scale Mode=::NewEnumerator4
    Sprite Size=<153.7841 20>
    Sprite Size Mode=::NewEnumerator4
    UsePositionOffset=false
    Write Material Random=true
   )
   /SetVariables(Particles.Light Cone Color:LinearColor=<rgba 7 0.2984 0.2188 1>, Particles.Light Cone Scale:Vector3f=<0.3 0.3 0.3>) {isSetParameters=true}
   /SetVariables(
    Particles.Cone Glow Color:LinearColor=<rgba 1 0.0684 0.0573 0.45>
    Particles.Cone Glow Size:Vector2f=&Vector2DFromFloat(Value=25)
   ) {isSetParameters=true}
  ]
  particleUpdateScript [
   &ParticleState()
   &ScaleMeshSize(Scale Factor=&MergeHelper_VectorSwitch(
     0=&VectorFromCurve(VectorCurve=DataInterface() {propertyValues {
        XCurve {keys [{arriveTangent=163.3652, interpMode=RCIM_Cubic, leaveTangent=163.3652, tangentMode=RCTM_User, time=0, value=1}, <at 0.1 1 RCIM_Cubic>]}
        YCurve {keys [<at 0 1>,<at 1 1>]}
        ZCurve {keys [{arriveTangent=163.3652, interpMode=RCIM_Cubic, leaveTangent=163.3652, tangentMode=RCTM_User, time=0, value=1}, <at 0.1 1 RCIM_Cubic>]}
       }})
     1=&VectorFromCurve(VectorCurve=DataInterface() {propertyValues {
        XCurve {keys [{arriveTangent=163.3652, interpMode=RCIM_Cubic, leaveTangent=163.3652, tangentMode=RCTM_User, time=0, value=1}, <at 0.1 1 RCIM_Cubic>]}
        YCurve {keys [<at 0 1>,<at 1 1>]}
        ZCurve {keys [{arriveTangent=163.3652, interpMode=RCIM_Cubic, leaveTangent=163.3652, tangentMode=RCTM_User, time=0, value=1}, <at 0.1 1 RCIM_Cubic>]}
       }})
     2=&VectorFromCurve(VectorCurve=DataInterface() {propertyValues {
        XCurve {keys [{arriveTangent=163.3652, interpMode=RCIM_Cubic, leaveTangent=163.3652, tangentMode=RCTM_User, time=0, value=1}, <at 0.1 1 RCIM_Cubic>]}
        YCurve {keys [<at 0 1>,<at 1 1>]}
        ZCurve {keys [{arriveTangent=163.3652, interpMode=RCIM_Cubic, leaveTangent=163.3652, tangentMode=RCTM_User, time=0, value=1}, <at 0.1 1 RCIM_Cubic>]}
       }})
     3=&VectorFromCurve(VectorCurve=DataInterface() {propertyValues {
        XCurve {keys [{arriveTangent=163.3652, interpMode=RCIM_Cubic, leaveTangent=163.3652, tangentMode=RCTM_User, time=0, value=1}, <at 0.1 1 RCIM_Cubic>]}
        YCurve {keys [<at 0 1>,<at 1 1>]}
        ZCurve {keys [{arriveTangent=163.3652, interpMode=RCIM_Cubic, leaveTangent=163.3652, tangentMode=RCTM_User, time=0, value=1}, <at 0.1 1 RCIM_Cubic>]}
       }})
     4=&VectorFromCurve(VectorCurve=DataInterface() {propertyValues {
        XCurve {keys [{arriveTangent=163.3652, interpMode=RCIM_Cubic, leaveTangent=163.3652, tangentMode=RCTM_User, time=0, value=1}, <at 0.1 1 RCIM_Cubic>]}
        YCurve {keys [<at 0 1>,<at 1 1>]}
        ZCurve {keys [{arriveTangent=163.3652, interpMode=RCIM_Cubic, leaveTangent=163.3652, tangentMode=RCTM_User, time=0, value=1}, <at 0.1 1 RCIM_Cubic>]}
       }})
     5=&VectorFromCurve(VectorCurve=DataInterface() {propertyValues {
        XCurve {keys [{arriveTangent=163.3652, interpMode=RCIM_Cubic, leaveTangent=163.3652, tangentMode=RCTM_User, time=0, value=1}, <at 0.1 1 RCIM_Cubic>]}
        YCurve {keys [<at 0 1>,<at 1 1>]}
        ZCurve {keys [{arriveTangent=163.3652, interpMode=RCIM_Cubic, leaveTangent=163.3652, tangentMode=RCTM_User, time=0, value=1}, <at 0.1 1 RCIM_Cubic>]}
       }})
     6=&VectorFromCurve(VectorCurve=DataInterface() {propertyValues {
        XCurve {keys [{arriveTangent=163.3652, interpMode=RCIM_Cubic, leaveTangent=163.3652, tangentMode=RCTM_User, time=0, value=1}, <at 0.1 1 RCIM_Cubic>]}
        YCurve {keys [<at 0 1>,<at 1 1>]}
        ZCurve {keys [{arriveTangent=163.3652, interpMode=RCIM_Cubic, leaveTangent=163.3652, tangentMode=RCTM_User, time=0, value=1}, <at 0.1 1 RCIM_Cubic>]}
       }})
     Index=&SpawnGroup()
     Num=::NewEnumerator5
    )) {module=ScaleMeshSize001}
   &ScaleMeshSize(Scale Factor=&VectorFromCurve(
     VectorCurve=DataInterface() {propertyValues {
       XCurve {keys [
         <at 0 0 RCIM_Cubic>
         {arriveTangentWeight=0.8186, interpMode=RCIM_Cubic, leaveTangentWeight=2.4558, tangentMode=RCTM_User, time=0.05, value=1}
         <at 0.2 1 RCIM_Cubic>
         {interpMode=RCIM_Cubic, leaveTangent=199.6162, leaveTangentWeight=0.0482, tangentMode=RCTM_Break, time=1.5, value=1}
         {arriveTangent=185.9243, interpMode=RCIM_Cubic, leaveTangent=185.9243, tangentMode=RCTM_User, time=1.5, value=1}
         {
          arriveTangent=-62.4247
          arriveTangentWeight=0.0442
          interpMode=RCIM_Cubic
          leaveTangent=-0.0277
          leaveTangentWeight=0.0335
          tangentMode=RCTM_Break
          time=1.5833
          value=1
         }
         {arriveTangentWeight=0.3, interpMode=RCIM_Cubic, tangentMode=RCTM_User, time=1.7033, value=0.5}
        ]}
       YCurve {keys [
         <at 0 1 RCIM_Cubic>
         <at 0.1199 1 RCIM_Cubic>
         {arriveTangentWeight=0.0471, interpMode=RCIM_Cubic, leaveTangentWeight=0.4243, tangentMode=RCTM_User, time=1.6333, value=1}
         {arriveTangentWeight=0.3, interpMode=RCIM_Cubic, tangentMode=RCTM_User, time=1.7033, value=0}
        ]}
       ZCurve {keys [
         {interpMode=RCIM_Cubic, leaveTangentWeight=0.3333, tangentMode=RCTM_User, time=0, value=0}
         {arriveTangentWeight=0.8971, interpMode=RCIM_Cubic, leaveTangentWeight=2.6912, tangentMode=RCTM_User, time=0.05, value=1}
         <at 0.2 1 RCIM_Cubic>
         {
          arriveTangentWeight=0.2677
          interpMode=RCIM_Cubic
          leaveTangent=199.6976
          leaveTangentWeight=0.0479
          tangentMode=RCTM_Break
          time=1.5
          value=1
         }
         {
          arriveTangent=-55.1202
          arriveTangentWeight=0.044
          interpMode=RCIM_Cubic
          leaveTangent=-0.0276
          leaveTangentWeight=0.0333
          tangentMode=RCTM_Break
          time=1.5833
          value=1
         }
         {arriveTangentWeight=0.3333, interpMode=RCIM_Cubic, tangentMode=RCTM_User, time=1.7033, value=0.5}
        ]}
      }}
     CurveIndex=$System.Age
    ))
   &SkeletalMeshLocation(
    Bone / Socket Index=&SelectIntFromArray(
     Direct Array Index=&SpawnGroup()
     Int Selection Array=DataInterface() {propertyValues {IntData [0,1,2,3,4,5,6]}}
    )
    Bone Sampling Mode=::NewEnumerator4
    MaskSpawnGroup=false
    Skeletal Mesh=$System.Skeletal Mesh Reader
   )
   /SetVariables(
    Particles.Cone Glow Color:LinearColor=&MultiplyLinearColor(
     Linear Color=$Particles.Initial.Cone Glow Color
     Scale Factor=&MakeVector4FromVectorAndFloat(
      "Vector (RGB)"=<1 1 1>
      "Float (Alpha)"=&FloatFromCurve(
       FloatCurve=DataInterface() {propertyValues {Curve {
          defaultValue=1
          keys [
           {arriveTangentWeight=0.5045, interpMode=RCIM_Cubic, tangentMode=RCTM_User, time=1.6333, value=1}
           {arriveTangentWeight=0.0233, interpMode=RCIM_Cubic, tangentMode=RCTM_User, time=1.7033, value=0}
          ]
         }}}
       CurveIndex=$System.Age
      )
     )
    )
    Particles.Light Cone Color:LinearColor=&MultiplyLinearColor(
     Linear Color=$Particles.Initial.Light Cone Color
     Scale Factor=&MakeVector4FromVectorAndFloat(
      "Vector (RGB)"=<1 1 1>
      "Float (Alpha)"=&FloatFromCurve(
       FloatCurve=DataInterface() {propertyValues {Curve {
          defaultValue=1
          keys [
           {arriveTangentWeight=0.5045, interpMode=RCIM_Cubic, tangentMode=RCTM_User, time=1.6333, value=1}
           {arriveTangentWeight=0.0233, interpMode=RCIM_Cubic, tangentMode=RCTM_User, time=1.7033, value=0}
          ]
         }}}
       CurveIndex=$System.Age
      )
     )
    )
   ) {isSetParameters=true}
  ]
  rendererClasses [&NiagaraMeshRendererProperties, &NiagaraSpriteRendererProperties]
  renderers [
   {rendererClass=&NiagaraMeshRendererProperties, rendererIndex=0}
   {rendererClass=&NiagaraMeshRendererProperties, rendererIndex=1}
   {rendererClass=&NiagaraMeshRendererProperties, rendererIndex=2}
   {rendererClass=&NiagaraMeshRendererProperties, rendererIndex=3}
   {rendererClass=&NiagaraSpriteRendererProperties, rendererIndex=4}
  ]
 }
 Light {rendererClasses [], renderers []}
 Motes {
  emitterUpdateScript [
   &EmitterState()
   &SpawnBurst_Instantaneous(Spawn Count=7, Spawn Time=$System.Mote Spawn Time, Use Spawn Probability=false, UseLoopCountLimit=false)
   &SpawnBurst_Instantaneous(Spawn Count=7, Spawn Group=1, Spawn Time=$System.Mote Spawn Time, Use Spawn Probability=false, UseLoopCountLimit=false) {module=SpawnBurst_Instantaneous002}
   &SpawnBurst_Instantaneous(Spawn Count=7, Spawn Group=2, Spawn Time=$System.Mote Spawn Time, Use Spawn Probability=false, UseLoopCountLimit=false) {module=SpawnBurst_Instantaneous003}
   &SpawnBurst_Instantaneous(Spawn Count=7, Spawn Group=3, Spawn Time=$System.Mote Spawn Time, Use Spawn Probability=false, UseLoopCountLimit=false) {module=SpawnBurst_Instantaneous004}
   &SpawnBurst_Instantaneous(Spawn Count=7, Spawn Group=4, Spawn Time=$System.Mote Spawn Time, Use Spawn Probability=false, UseLoopCountLimit=false) {module=SpawnBurst_Instantaneous005}
   &SpawnBurst_Instantaneous(Spawn Count=7, Spawn Group=5, Spawn Time=$System.Mote Spawn Time, Use Spawn Probability=false, UseLoopCountLimit=false) {module=SpawnBurst_Instantaneous006}
   &SpawnBurst_Instantaneous(Spawn Count=7, Spawn Group=6, Spawn Time=$System.Mote Spawn Time, Use Spawn Probability=false, UseLoopCountLimit=false) {module=SpawnBurst_Instantaneous007}
  ]
  particleSpawnScript [
   &InitializeParticle(
    Color Maximum=<rgba 3 0.0737 0.0625 0.2>
    Color Minimum=<rgba 3 0.0737 0.0625 0.5>
    Color Mode=::NewEnumerator2
    Lifetime=&RandomRangeFloat(Minimum=0.2, Maximum=0.35)
    Mesh Scale=&MergeHelper_VectorSwitch(
     0=<0.1 0.46 0.1>
     1=<0.1 0.45 0.1>
     2=<0.1 0.5 0.1>
     3=<0.1 0.5 0.1>
     4=<0.1 0.4 0.1>
     5=<0.1 0.42 0.1>
     6=<0.1 0.45 0.1>
     Index=&SpawnGroup()
     Num=::NewEnumerator5
    )
    Mesh Scale Mode=::NewEnumerator4
    Sprite Size Mode=::NewEnumerator1
    Uniform Sprite Size Max=3
    Uniform Sprite Size Min=1
    UsePositionOffset=false
    Write Material Random=true
   )
   &SkeletalMeshLocation(
    Bone / Socket Index=&SelectIntFromArray(
     Direct Array Index=&SpawnGroup()
     Int Selection Array=DataInterface() {propertyValues {IntData [0,1,2,3,4,5,6]}}
    )
    Bone Sampling Mode=::NewEnumerator4
    MaskSpawnGroup=false
    Orientation Sampling=::NewEnumerator1
    Position Sampling=::NewEnumerator2
    Sampled Position Offset=<0 -85 0>
    Skeletal Mesh=$System.Skeletal Mesh Reader
   )
   &ShapeLocation(
    Additional Quaternion Rotation=true
    Additional Yaw / Pitch / Roll=<0 0 0.75>
    Cylinder Height=70
    Cylinder Radius=1
    MaskSpawnGroup=false
    Offset Mode=::NewEnumerator0
    Override Seed=false
    Rotation Mode=::NewEnumerator3
    Rotation Quaternion=$Particles.SkeletalMeshLocation.SampledRotation
    Shape Origin=$Particles.SkeletalMeshLocation.SampledPosition
    Shape Primitive=::NewEnumerator1
    Surface Only=false
    UseOwnerScale=true
   )
   &AddVelocity(
    Additional Quaternion Rotation=false
    Rotation Mode=::NewEnumerator3
    Rotation Quaternion:Quat4f=$Particles.SkeletalMeshLocation.SampledRotation
    Velocity=<0 100 0>
    Velocity Speed Scale=&RandomRangeFloat(Minimum=1, Maximum=3)
   )
  ]
  particleUpdateScript [
   &ParticleState()
   &Drag(
    Drag=&FloatFromCurve(
     FloatCurve=DataInterface() {propertyValues {Curve {keys [<at 0 0.5>,<at 0.7 1>]}}}
     Scale Curve=&RandomRangeFloat(Minimum=18, Maximum=36)
    )
    Rotational Drag:NiagaraFloat=&RandomRangeFloat(Minimum=3, Maximum=5)
    Use Linear Drag=true
    Use Rotational Drag=true
   )
   &ScaleVelocity(Velocity Scale=&VectorFromCurve(VectorCurve=DataInterface() {propertyValues {
       XCurve {keys [<at 0 1 RCIM_Cubic>, {arriveTangentWeight=0.3938, interpMode=RCIM_Cubic, tangentMode=RCTM_User, time=0.7, value=0.3}]}
       YCurve {keys [<at 0 1 RCIM_Cubic>, {arriveTangentWeight=0.3938, interpMode=RCIM_Cubic, tangentMode=RCTM_User, time=0.7, value=0.3}]}
       ZCurve {keys [<at 0 1 RCIM_Cubic>, {arriveTangentWeight=0.3938, interpMode=RCIM_Cubic, tangentMode=RCTM_User, time=0.7, value=0.3}]}
      }}))
   &SolveForcesAndVelocity(Clamp Velocity=false, Limit Acceleration=false)
   &ScaleSpriteSizeBySpeed(Max Scale Factor=<1 10>, Velocity Threshold=250)
   &ScaleColor(
    Linear Color Curve=DataInterface() {propertyValues {
      AlphaCurve {keys [<at 0 1>,<at 0.5954 1>,<at 1 0>]}
      BlueCurve {keys [<at 0 1>,<at 1 1>]}
      GreenCurve {keys [<at 0 1>,<at 1 1>]}
      RedCurve {keys [<at 0 1>,<at 1 1>]}
     }}
    Scale Mode=2
   ) {module=ScaleColor001}
  ]
  rendererClasses [&NiagaraSpriteRendererProperties]
  renderers [{rendererClass=&NiagaraSpriteRendererProperties}]
 }
 Wave {
  emitterSpawnScript [/SetVariables(Emitter.Size=1.1) {isSetParameters=true}]
  emitterUpdateScript [
   &EmitterState(Life Cycle Mode=::NewEnumerator1, Loop Behavior=::NewEnumerator1, Loop Duration=3, UseLoopDelay=false)
   /SetVariables(Emitter.Center Mask Radius:NiagaraFloat=&FloatFromCurve(FloatCurve=DataInterface() {propertyValues {Curve {keys [<at 0.4918 0>,<at 0.7541 1>]}}}, CurveIndex=$Emitter.NormalizedLoopAge)) {isSetParameters=true}
   &SpawnBurst_Instantaneous(Spawn Time=1.5, Use Spawn Probability=false, UseLoopCountLimit=false) {module=SpawnBurst_Instantaneous001}
  ]
  particleSpawnScript [
   &InitializeParticle(
    Color=<rgba 4 4 4 1>
    Lifetime=0.35
    Mesh Scale Mode=::NewEnumerator3
    Mesh Uniform Scale=&Multiply_Float(A=0.3, B=$Emitter.Size)
    Position Offset=<-10 20 100>
    Sprite Rotation Angle=&SelectFloatFromArray(
     Array Sampling Mode=::NewEnumerator0
     Direct Array Index=?
     Float Selection Array=DataInterface() {propertyValues {FloatData [0,0.25,0.5,0.75]}}
     Random Seed=0
    )
    Sprite Rotation Mode=::NewEnumerator1
    Sprite Size Mode=::NewEnumerator3
    Uniform Sprite Size=&Multiply_Float(A=100, B=$Emitter.Size)
    UsePositionOffset=true
    Write Material Random=true
   )
   /SetVariables(Particles.VisibilityTag=$Emitter.SpawnGroup) {isSetParameters=true}
   /SetVariables(Particles.WaveColor:LinearColor=<rgba 5 0 0 1>, Particles.WaveLifetime=0.6) {isSetParameters=true}
   /SetVariables(
    Particles.Glow Lifetime=0.2
    Particles.Glow Size=&Multiply_Vector2D_ByFloat(Float=$Emitter.Size, Vector2D=&Multiply_Vector2D_ByFloat(Float=50, Vector2D=<1 1>))
   ) {isSetParameters=true}
   &InitialMeshOrientation(
    Mesh Orientation Mode=::NewEnumerator2
    Orientation Vector=<0 0 1>
    Rotation=<0 0 0.5>
    Rotation Coordinate Space=::NewEnumerator0
    Use Rotation Vector=false
   )
   /SetVariables(Particles.Circuit Lifetime=0.4) {isSetParameters=true}
   &SpriteFacingAndAlignment(Sprite Alignment=<1 0 0>, Sprite Facing=<0 1 0>, WriteAlignment=true, WriteFacing=true)
  ]
  particleUpdateScript [
   &ParticleState()
   /SetVariables(Particles.Update Circuit Age:NiagaraFloat=&Divide_Float(A=$Particles.Age, B=$Particles.Circuit Lifetime)) {isSetParameters=true}
   /SetVariables(Particles.Update WaveAge:NiagaraFloat=&Divide_Float(A=$Particles.Age, B=$Particles.WaveLifetime)) {isSetParameters=true}
   /SetVariables(Particles.Update Glow Age:NiagaraFloat=&Divide_Float(A=$Particles.Age, B=$Particles.Glow Lifetime)) {isSetParameters=true}
   &ScaleColor(
    Scale Mode=::NewEnumerator1
    Scale RGBA=&Vector4FromCurve(
     CurveIndex=$Particles.Update Circuit Age
     Vector4Curve=DataInterface() {propertyValues {
       AlphaCurve {keys [<at 0 1>,<at 1 0>]}
       BlueCurve {keys [<at 0 0>,<at 1 0>]}
       GreenCurve {keys [<at 0 0>,<at 1 0>]}
       RedCurve {keys [<at 0 1>,<at 1 0.2>]}
      }}
    )
    ScaleRGBA=true
   )
   &ScaleSpriteSize(Scale Factor=&Vector2DFromFloat(Value=&FloatFromCurve(
      FloatCurve=DataInterface() {propertyValues {Curve {keys [
          {arriveTangent=0.5497, interpMode=RCIM_Cubic, leaveTangent=0.5497, tangentMode=RCTM_User, time=0, value=0.4128}
          {arriveTangent=0.5872, interpMode=RCIM_Cubic, leaveTangent=0.5872, time=0.2754, value=0.8613}
          {interpMode=RCIM_Cubic, tangentMode=RCTM_User, time=1, value=1}
         ]}}}
      CurveIndex=$Particles.Update Circuit Age
     )))
   &DynamicMaterialParameters(
    Index 0 Param 2=&FloatFromCurve(
     FloatCurve=DataInterface() {propertyValues {Curve {keys [
         <at 0 1 RCIM_Cubic>
         {arriveTangent=0.1115, interpMode=RCIM_Cubic, leaveTangent=0.1115, tangentMode=RCTM_User, time=0.1, value=0.5}
         {interpMode=RCIM_Cubic, tangentMode=RCTM_User, time=1, value=1}
        ]}}}
     CurveIndex=$Particles.Update Circuit Age
    )
    Module.Index0Float0Write=true
    Module.Index0Float1Write=true
    Module.Index0Float2Write=true
    Module.Index0Float3Write=true
   )
   /SetVariables(Particles.DY0:Vector4f=&MakeVector4(
     X=0
     Y=&FloatFromCurve(FloatCurve=DataInterface() {propertyValues {Curve {keys [{tangentMode=RCTM_User, time=0, value=0.47}, {tangentMode=RCTM_User, time=1, value=1}]}}})
     Z=0
     W=0
    )) {isSetParameters=true}
   &ScaleMeshSize(Scale Factor=&VectorFromFloat(Value=&FloatFromCurve(
      FloatCurve=DataInterface() {propertyValues {Curve {keys [
          {arriveTangent=4.8748, interpMode=RCIM_Cubic, leaveTangent=4.8748, tangentMode=RCTM_User, time=0, value=1}
          {arriveTangent=0.9672, interpMode=RCIM_Cubic, leaveTangent=0.9672, tangentMode=RCTM_User, time=0.3099, value=1.9838}
          {arriveTangent=0.1926, interpMode=RCIM_Cubic, leaveTangent=0.1926, tangentMode=RCTM_User, time=1, value=2.1883}
         ]}}}
      CurveIndex=$Particles.Update WaveAge
     )))
   /SetVariables(
    Particles.Update Wave Color:LinearColor=&MakeLinearColorFromVectorAndFloat(
     "Vector (RGB)"=&Multiply_VectorByFloat(
      Vector=&MakeVectorFromLinearColorRGB(Color=$Particles.WaveColor)
      Float=&FloatFromCurve(
       FloatCurve=DataInterface() {propertyValues {Curve {
          defaultValue=1
          keys [
           {tangentMode=RCTM_User, time=0, value=1}
           {arriveTangent=-0.0405, interpMode=RCIM_Cubic, leaveTangent=-0.0405, tangentMode=RCTM_User, time=0.7009, value=1}
           {arriveTangent=-8.0089, interpMode=RCIM_Cubic, leaveTangent=-8.0089, tangentMode=RCTM_User, time=1, value=0}
          ]
         }}}
       Scale Curve=15
      )
     )
     "Float (Alpha)"=&FloatFromCurve(
      FloatCurve=DataInterface() {propertyValues {Curve {keys [
          <at 0 0 RCIM_Cubic>
          {arriveTangent=0.4167, interpMode=RCIM_Cubic, leaveTangent=0.4167, time=0.1, value=1}
          {arriveTangent=-1.1111, interpMode=RCIM_Cubic, leaveTangent=-1.1111, time=0.6, value=0.25}
          <at 1 0 RCIM_Cubic>
         ]}}}
      CurveIndex=$Particles.Update WaveAge
      Scale Curve=0.05
     )
    )
    Particles.Update Wave DynamicParameter:Vector4f=&MakeVector4(
     X=&FloatFromCurve(FloatCurve=DataInterface() {propertyValues {Curve {keys [<at 0 1>,<at 1 0>]}}}, CurveIndex=$Particles.Update WaveAge)
     Y=2
     Z=&FloatFromCurve(FloatCurve=DataInterface() {propertyValues {Curve {keys [<at 0.1 1>,<at 1 0>]}}}, CurveIndex=$Particles.Update WaveAge, Scale Curve=6)
     W=&FloatFromCurve(
      FloatCurve=DataInterface() {propertyValues {Curve {keys [
          <at 0 1>
          {arriveTangent=0.3079, interpMode=RCIM_Cubic, leaveTangent=0.3079, tangentMode=RCTM_User, time=0.3, value=0.8}
          <at 1 1 RCIM_Cubic>
         ]}}}
      CurveIndex=$Particles.Update WaveAge
     )
    )
    Particles.Update Wave DynamicParameter001:Vector4f=&MakeVector4(
     X=&FN_PickFloatDependOnPlatform(TargetPlatformActive=180, TargetPlatformNoActive=120)
     Y=&FN_PickFloatDependOnPlatform(TargetPlatformActive=2.5, TargetPlatformNoActive=4)
     Z=0
     W=0
    )
   ) {isSetParameters=true}
   /SetVariables(
    Particles.Update Glow Color:LinearColor=&MakeLinearColorFromVectorAndFloat(
     "Vector (RGB)"=&MakeVectorFromLinearColorRGB(Color=<rgba 1 0.0833 0.0833 1>)
     "Float (Alpha)"=&FloatFromCurve(
      FloatCurve=DataInterface() {propertyValues {Curve {
         defaultValue=1
         keys [
          {interpMode=RCIM_Cubic, tangentMode=RCTM_User, time=0, value=0}
          {interpMode=RCIM_Cubic, tangentMode=RCTM_User, time=0.2399, value=1}
          {tangentMode=RCTM_User, time=1, value=0}
         ]
        }}}
      CurveIndex=$Particles.Update Glow Age
     )
    )
    Particles.Update Glow DynamicParameter:Vector4f=<2 0 0 0>
    Particles.Update Glow Size:Vector2f=&Multiply_Vector2D(
     A=$Particles.Glow Size
     B=&Multiply_Vector2D_ByFloat(
      Float=&FloatFromCurve(
       FloatCurve=DataInterface() {propertyValues {Curve {
          defaultValue=0.367
          keys [
           {interpMode=RCIM_Cubic, tangentMode=RCTM_User, time=0, value=0.367}
           {arriveTangent=-0.1298, interpMode=RCIM_Cubic, leaveTangent=-0.1298, tangentMode=RCTM_User, time=1, value=3.8846}
          ]
         }}}
       CurveIndex=$Particles.Update Glow Age
      )
      Vector2D=<1.25 1.25>
     )
    )
   ) {isSetParameters=true}
  ]
  rendererClasses [&NiagaraMeshRendererProperties, &NiagaraSpriteRendererProperties]
  renderers [
   {rendererClass=&NiagaraMeshRendererProperties, rendererIndex=0}
   {rendererClass=&NiagaraSpriteRendererProperties, rendererIndex=1}
  ]
 }
}
summary {
 systemName=NS_Emote_ScanEvade_Lasers
 emitters [
  {emitterName=Laser, rendererClasses [&NiagaraMeshRendererProperties, &NiagaraSpriteRendererProperties]}
  {emitterName=Motes, rendererClasses [&NiagaraSpriteRendererProperties]}
  {emitterName=Wave, rendererClasses [&NiagaraMeshRendererProperties, &NiagaraSpriteRendererProperties]}
  {emitterName=Light, rendererClasses []}
 ]
 userVariables []
}
