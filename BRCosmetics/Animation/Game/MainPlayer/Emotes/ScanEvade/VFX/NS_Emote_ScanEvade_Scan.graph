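# Niagara system for the scan VFX of the ScanEvade emote: a short-lived, mostly green scan.
# Grid draws opposite-facing sprites and meshes that snap open and shut via scale curves, and writes decal attributes.
# Three directional glow sprites and a fading PlayerOverlayWipe mesh accompany it, while rising sparks spawn in windows at start and end.
# The glow emitter also writes light attributes tinted by User.ColorIndex and time of day; Light and Spawn_Mesh are empty.
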NiagaraSystem /BRCosmetics/Animation/Game/MainPlayer/Emotes/ScanEvade/VFX/NS_Emote_ScanEvade_Scan
dependencies {
 usedDataInterfaces [
  &NiagaraDataInterfaceCurve
  &NiagaraDataInterfaceVector2DCurve
  &NiagaraDataInterfaceVectorCurve
  &NiagaraDataInterfaceColorCurve
 ]
 usedDynamicInputs []
 usedModules [
  &self:SystemSpawnScript.SystemScriptSource.SystemScriptGraph.NiagaraNodeAssignment_1.SetVariables_#1
  &SystemState
 ]
}
emitters {
 Grid {
  emitterUpdateScript [
   &EmitterState()
   &SpawnBurst_Instantaneous(Spawn Count=2, Spawn Group=1, Spawn Time=0.1, Use Spawn Probability=false, UseLoopCountLimit=false) {module=SpawnBurst_Instantaneous001}
   &SpawnBurst_Instantaneous(Spawn Time=0.1, Use Spawn Probability=false, UseLoopCountLimit=false)
  ]
  particleSpawnScript [
   &InitializeParticle(
    Color=<rgba 0.1471 1 0.0123 0.7>
    Lifetime=1.45
    Mesh Scale Mode=::NewEnumerator3
    Sprite Size=<200 200>
    Sprite Size Mode=::NewEnumerator4
    UsePositionOffset=false
    Write Material Random=true
   )
   &PartitionParticles()
   &InitialMeshOrientation(
    Mesh Orientation Mode=::NewEnumerator2
    Orientation Vector=&MergeHelper_VectorSwitch(0=<1 0 0>, 1=<-1 0 0>, 2=?, 3=?, Index=$Particles.PartitionParticles.Partition, Num=::NewEnumerator0)
    Use Rotation Vector=false
   )
   /SetVariables(Particles.Glow Color=<rgba 0.1724 1 0.0417 0.3>, Particles.Glow Scale:Vector3f=<1 1 1>) {isSetParameters=true}
   &SpriteFacingAndAlignment(
    Sprite Alignment=<0 0 1>
    Sprite Facing=&MergeHelper_VectorSwitch(0=<1 0 0>, 1=<-1 0 0>, 2=?, 3=?, Index=$Particles.PartitionParticles.Partition, Num=::NewEnumerator0)
    WriteAlignment=true
    WriteFacing=true
   )
   /SetVariables(Particles.VisibilityTag=&SpawnGroup()) {isSetParameters=true}
   /SetVariables(Particles.Scan Mesh Color:LinearColor=<rgba 0.1471 1 0.0123 0.2>) {isSetParameters=true}
  ]
  particleUpdateScript [
   &ParticleState()
   &ScaleSpriteSize(
    Non-Uniform Curve Sprite Scale=DataInterface() {propertyValues {
      XCurve {keys [{tangentMode=RCTM_User, time=0, value=1}, {tangentMode=RCTM_User, time=1, value=1}]}
      YCurve {
       defaultValue=0
       keys [
        {arriveTangent=23.9764, interpMode=RCIM_Cubic, leaveTangent=23.9764, tangentMode=RCTM_User, time=0, value=0}
        {arriveTangentWeight=0.05, interpMode=RCIM_Cubic, tangentMode=RCTM_User, time=0.15, value=1}
        {arriveTangentWeight=0.2333, interpMode=RCIM_Cubic, tangentMode=RCTM_User, time=0.85, value=1}
        {arriveTangent=-24.3584, arriveTangentWeight=0.05, interpMode=RCIM_Cubic, leaveTangent=-24.3584, tangentMode=RCTM_User, time=1, value=0}
       ]
      }
     }}
    Scale Sprite Size Mode=::NewEnumerator3
   )
   &ScaleMeshSize(Scale Factor=&VectorFromCurve(VectorCurve=DataInterface() {propertyValues {
       XCurve {keys [
         {arriveTangentWeight=0.0167, interpMode=RCIM_Cubic, leaveTangentWeight=0.3, tangentMode=RCTM_User, time=0.15, value=1}
         {arriveTangentWeight=0.3986, interpMode=RCIM_Cubic, leaveTangentWeight=0.0498, tangentMode=RCTM_User, time=0.85, value=1}
        ]}
       YCurve {
        defaultValue=1
        keys [
         {arriveTangent=23.9764, interpMode=RCIM_Cubic, leaveTangent=23.9764, tangentMode=RCTM_User, time=0, value=0}
         {arriveTangentWeight=0.05, interpMode=RCIM_Cubic, tangentMode=RCTM_User, time=0.15, value=1}
         {arriveTangentWeight=0.2333, interpMode=RCIM_Cubic, tangentMode=RCTM_User, time=0.85, value=1}
         {arriveTangent=-24.3584, arriveTangentWeight=0.05, interpMode=RCIM_Cubic, leaveTangent=-24.3584, tangentMode=RCTM_User, time=1, value=0}
        ]
       }
       ZCurve {keys [<at 0 1>,<at 0.1708 1>,<at 0.8431 1>,<at 1 1>]}
      }}))
   /Niagara/Modules/Decal/Decal_Attributes(
    Color=1
    Color Value To Scale:LinearColor=&ColorFromCurve(ColorCurve=DataInterface() {propertyValues {
       AlphaCurve {keys [<at 0 0>, <at 0.1058 1>, {arriveTangent=-1.1183, interpMode=RCIM_Cubic, leaveTangent=-1.1183, time=0.8478, value=1}, <at 1 0 RCIM_Cubic>]}
       BlueCurve {keys [<at 0 0.3958 RCIM_Cubic>]}
       GreenCurve {keys [<at 0 1 RCIM_Cubic>]}
       RedCurve {keys [<at 0 0.5012 RCIM_Cubic>]}
      }})
    Decal Position Offset:Vector3f=<0 0 100>
    Decal Size:Vector3f=<200 5 200>
    Position=1
    Scale Alpha:NiagaraFloat=&FloatFromCurve(
     FloatCurve=DataInterface() {propertyValues {Curve {keys [
         {interpMode=RCIM_Cubic, tangentMode=RCTM_User, time=0, value=0}
         {
          arriveTangent=0.1116
          arriveTangentWeight=0.4713
          interpMode=RCIM_Cubic
          leaveTangent=-0.3115
          leaveTangentWeight=0.2344
          tangentMode=RCTM_Break
          tangentWeightMode=RCTWM_WeightedBoth
          time=0.5
          value=1
         }
         {arriveTangent=-7.7147, interpMode=RCIM_Cubic, leaveTangent=-7.7147, tangentMode=RCTM_User, time=1, value=0}
        ]}}}
     Scale Curve=0.1
    )
    ScaleA=false
    ScaleRGB=false
    Size=1
    UsePositionOffset=true
   )
   /SetVariables() {enabled=false, isSetParameters=true}
   /SetVariables(Particles.Fresnel DynamicMaterial Parameter 1:Vector4f=<1 1 22.9087 1>) {enabled=false, isSetParameters=true}
   /SetVariables(Particles.Scan Mesh Scale:Vector3f=&VectorFromCurve(VectorCurve=DataInterface() {propertyValues {
       XCurve {keys [
         {arriveTangentWeight=0.0167, interpMode=RCIM_Cubic, leaveTangentWeight=0.3, tangentMode=RCTM_User, time=0.15, value=1}
         {arriveTangentWeight=0.3986, interpMode=RCIM_Cubic, leaveTangentWeight=0.0498, tangentMode=RCTM_User, time=0.85, value=1}
        ]}
       YCurve {
        defaultValue=1
        keys [
         {arriveTangent=23.9764, interpMode=RCIM_Cubic, leaveTangent=23.9764, tangentMode=RCTM_User, time=0, value=0}
         {arriveTangentWeight=0.05, interpMode=RCIM_Cubic, tangentMode=RCTM_User, time=0.15, value=1}
         {arriveTangentWeight=0.2333, interpMode=RCIM_Cubic, tangentMode=RCTM_User, time=0.85, value=1}
         {arriveTangent=-24.3584, arriveTangentWeight=0.05, interpMode=RCIM_Cubic, leaveTangent=-24.3584, tangentMode=RCTM_User, time=1, value=0}
        ]
       }
       ZCurve {keys [<at 0 1>,<at 0.1708 1>,<at 0.8431 1>,<at 1 1>]}
      }})) {isSetParameters=true}
  ]
  rendererClasses [&NiagaraSpriteRendererProperties, &NiagaraMeshRendererProperties]
  renderers [
   {rendererClass=&NiagaraSpriteRendererProperties}
   {rendererClass=&NiagaraSpriteRendererProperties, rendererIndex=1}
   {rendererClass=&NiagaraMeshRendererProperties, rendererIndex=2}
   {rendererClass=&NiagaraMeshRendererProperties, rendererIndex=3}
   {rendererClass=&NiagaraMeshRendererProperties, rendererIndex=4}
  ]
 }
 Light {rendererClasses [], renderers []}
 Multidirectional_Glow {
  emitterUpdateScript [
   &EmitterState()
   &SpawnBurst_Instantaneous(Spawn Count=3, Spawn Time=0.1, Use Spawn Probability=false, UseLoopCountLimit=false)
  ]
  particleSpawnScript [
   &FN_SetSpawnGroup(Number of Spawn Groups=3)
   &InitializeParticle(
    Color=<rgba 1 1 1 0.35>
    Lifetime=1.45
    Position Offset=&MergeHelper_VectorSwitch(0=<0 112 0>, 1=<0 0 0>, 2=<0 -112 0>, 3=?, Index=&SpawnGroup(), Num=::NewEnumerator1)
    Sprite Size=&MergeHelper_Vector2Switch(0=<40 320>, 1=<350 220>, 2=<40 320>, 3=?, 4=?, Index=$Particles.SpawnGroup, Num=::NewEnumerator1)
    Sprite Size Mode=::NewEnumerator4
    UsePositionOffset=true
    Write Material Random=true
   )
   &CameraOffset(Camera Offset Amount=5)
   &SpriteFacingAndAlignment(
    Sprite Alignment=&MergeHelper_VectorSwitch(0=<0 0 1>, 1=<0 0 1>, 2=<0 0 1>, 3=?, Index=$Particles.SpawnGroup, Num=::NewEnumerator1)
    Sprite Facing=&MergeHelper_VectorSwitch(0=<0 1 1>, 1=<1 0 0>, 2=<0 -1 1>, 3=?, Index=$Particles.SpawnGroup, Num=::NewEnumerator1)
    WriteAlignment=true
    WriteFacing=true
   )
  ]
  particleUpdateScript [
   &ParticleState()
   &ScaleColor(
    Scale Alpha=&FloatFromCurve(
     FloatCurve=DataInterface() {propertyValues {Curve {
        defaultValue=0.8
        keys [
         {arriveTangent=23.9764, interpMode=RCIM_Cubic, leaveTangent=23.9764, tangentMode=RCTM_User, time=0, value=0}
         {arriveTangentWeight=0.05, interpMode=RCIM_Cubic, tangentMode=RCTM_User, time=0.15, value=1}
         {arriveTangentWeight=0.2333, interpMode=RCIM_Cubic, tangentMode=RCTM_User, time=0.85, value=1}
         {arriveTangent=-24.3584, arriveTangentWeight=0.05, interpMode=RCIM_Cubic, leaveTangent=-24.3584, tangentMode=RCTM_User, time=1, value=0}
        ]
       }}}
     Scale Curve=&MergeHelper_FloatSwitch(0=0.35, 1=0.4, 2=0.35, 3=?, Index=$Particles.SpawnGroup, Num=::NewEnumerator1)
    )
    Scale RGB=&Multiply_VectorByFloat(Vector=&MakeVectorFromLinearColorRGB(Color=<rgba 0 1 0.0938 1>), Float=1.25)
    ScaleA=true
    ScaleRGB=true
   )
   &FN_Light_Attributes(
    Color=&MultiplyLinearColor(
     Linear Color=&LerpLinearColors(StartColor=<rgba 0.0515 1 0.7773 1>, EndColor=<rgba 0.3649 0.0515 1 1>, LerpFactor=$User.ColorIndex)
     Scale Factor=&MakeVector4_FromLinearColor(LinearColor=$System.TODColor)
    )
    Create Color Attribute=true
    Create Exponent Attribute=true
    Create Position Attribute=true
    Create Radius Attribute=true
    Exponent=2
    Position=&AddVectorToPosition(Position=$Engine.Emitter.SimulationPosition, Vector=<0 -15 0>)
    Radius=30
    Scale Color=&Multiply_Float(
     A=5
     B=&FNTimeOfDayScalar(Night Scale=&Lerp_Float(A=1, B=2.5, Alpha=$User.ColorIndex), Morning Scale=1, Evening Scale=1)
    )
   )
  ]
  rendererClasses [&NiagaraSpriteRendererProperties]
  renderers [{rendererClass=&NiagaraSpriteRendererProperties}]
 }
 PlayerOverlayWipe {
  emitterUpdateScript [
   &EmitterState()
   &SpawnBurst_Instantaneous(Spawn Time=0.1, Use Spawn Probability=false, UseLoopCountLimit=false)
  ]
  particleSpawnScript [&InitializeParticle(
    Color=<rgba 0.014 1 0.0035 4.3221>
    Lifetime=1.45
    Mesh Scale=<15 204.7166 192.5>
    Mesh Scale Mode=::NewEnumerator4
    UsePositionOffset=true
    Write Material Random=true
   )]
  particleUpdateScript [
   &ParticleState()
   /SetVariables() {isSetParameters=true}
   &Color(Color=&ColorFromCurve(
     ColorCurve=DataInterface() {propertyValues {
       AlphaCurve {keys [<at 0 0>,<at 0.1901 1 RCIM_Cubic>,<at 0.8095 1>,<at 1 0 RCIM_Cubic>]}
       BlueCurve {keys [<at 0 1 RCIM_Cubic>,<at 1 1 RCIM_Cubic>]}
       GreenCurve {keys [<at 0 1 RCIM_Cubic>,<at 1 1 RCIM_Cubic>]}
       RedCurve {keys [<at 0 1 RCIM_Cubic>,<at 1 1 RCIM_Cubic>]}
      }}
     CurveIndex=$System.NormalizedLoopAge
    ))
  ]
  rendererClasses [&NiagaraMeshRendererProperties]
  renderers [{rendererClass=&NiagaraMeshRendererProperties, rendererIndex=0}]
 }
 Spawn_Mesh {rendererClasses [], renderers []}
 Spawn_Particles {
  emitterUpdateScript [
   &EmitterState()
   &SpawnRate(
    SpawnRate=&FloatFromCurve(
     FloatCurve=DataInterface() {propertyValues {Curve {keys [<at 0 1 RCIM_Constant>,<at 0.15 0 RCIM_Constant>,<at 1.41 1 RCIM_Constant>,<at 1.56 0 RCIM_Constant>]}}}
     CurveIndex=$System.Age
     Scale Curve=100
    )
    Use Spawn Probability=false
   )
  ]
  particleSpawnScript [
   &InitializeParticle(
    Color Mode=::NewEnumerator0
    Lifetime=&RandomRangeFloat(Minimum=0.2, Maximum=0.3)
    Sprite Size Mode=::NewEnumerator3
    Uniform Sprite Size=&RandomRangeFloat(Minimum=1, Maximum=3)
    UsePositionOffset=false
    Write Material Random=true
   )
   &CameraOffset()
   &AddVelocity(Velocity=&RandomRangeVector(Minimum=<0 0 250>, Maximum=<0 0 400>))
   &ShapeLocation(
    Box Size=<20 190 20>
    MaskSpawnGroup=false
    Offset=<0 0 16.7696>
    Offset Mode=::NewEnumerator0
    Override Seed=false
    Shape Primitive=::NewEnumerator2
    Surface Only=false
    UseOwnerScale=true
   )
  ]
  particleUpdateScript [
   &ParticleState()
   &ScaleSpriteSize(Uniform Curve Sprite Scale=DataInterface() {propertyValues {Curve {keys [
        <at 0 0 RCIM_Cubic>
        {arriveTangent=1.6667, interpMode=RCIM_Cubic, leaveTangent=1.6667, time=0.2, value=1}
        {arriveTangent=-1.25, interpMode=RCIM_Cubic, leaveTangent=-1.25, time=0.6, value=1}
        <at 1 0 RCIM_Cubic>
       ]}}})
   &Drag(
    Drag=&FloatFromCurve(
     FloatCurve=DataInterface() {propertyValues {Curve {keys [<at 0 0 RCIM_Cubic>, {arriveTangent=1.5074, interpMode=RCIM_Cubic, leaveTangent=1.5074, tangentMode=RCTM_User, time=1, value=1}]}}}
     Scale Curve=500
    )
    Rotational Drag:NiagaraFloat=&FloatFromCurve(
     FloatCurve=DataInterface() {propertyValues {Curve {keys [<at 0 0 RCIM_Cubic>, {arriveTangent=1.5074, interpMode=RCIM_Cubic, leaveTangent=1.5074, tangentMode=RCTM_User, time=1, value=1}]}}}
     Scale Curve=500
    )
    Use Linear Drag=true
    Use Rotational Drag=true
   )
   &SolveForcesAndVelocity(Clamp Velocity=false, Limit Acceleration=false)
   &ScaleColor(
    Linear Color Curve=DataInterface() {propertyValues {
      AlphaCurve {keys [<at 0 0>,<at 0.3343 1>,<at 1 0>]}
      BlueCurve {keys [<at 0 0.3437>]}
      GreenCurve {keys [<at 0 3>]}
      RedCurve {keys [<at 0 0.7062>]}
     }}
    Scale Mode=2
   )
  ]
  rendererClasses [&NiagaraSpriteRendererProperties]
  renderers [{rendererClass=&NiagaraSpriteRendererProperties}]
 }
}
summary {
 systemName=NS_Emote_ScanEvade_Scan
 emitters [
  {emitterName=Grid, rendererClasses [&NiagaraSpriteRendererProperties, &NiagaraMeshRendererProperties]}
  {emitterName=PlayerOverlayWipe, rendererClasses [&NiagaraMeshRendererProperties]}
  {emitterName=Light, rendererClasses []}
  {emitterName=Multidirectional_Glow, rendererClasses [&NiagaraSpriteRendererProperties]}
  {emitterName=Spawn_Mesh, rendererClasses []}
  {emitterName=Spawn_Particles, rendererClasses [&NiagaraSpriteRendererProperties]}
 ]
 userVariables [
  User.bIsFrontEnd {type=NiagaraBool}
  User.bIsFrontEndPreview {type=NiagaraBool}
  User.ColorIndex {type=NiagaraFloat, defaultValue=0, description=""}
  User.WipeDuration {type=NiagaraFloat, defaultValue=3.42, description=""}
 ]
}
