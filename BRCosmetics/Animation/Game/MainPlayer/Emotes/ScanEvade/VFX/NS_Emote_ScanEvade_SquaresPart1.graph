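# Niagara system for part one of the ScanEvade player emote: timed glowing squares, an inner cross, connecting lines and lights.
# Squares, cross and outer square follow skeletal mesh sockets, appearing in staggered bursts with curve-driven scale and flicker.
# Spline_Position passes socket-offset positions through system position arrays to scratch logic in Splines, which stretches line meshes between consecutive indices.
# System-age curves switch squares and a light from green to red.
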NiagaraSystem /BRCosmetics/Animation/Game/MainPlayer/Emotes/ScanEvade/VFX/NS_Emote_ScanEvade_SquaresPart1
dependencies {
 usedDataInterfaces [
  &NiagaraDataInterfaceColorCurve
  &NiagaraDataInterfaceCurve
  &NiagaraDataInterfaceVector2DCurve
  &NiagaraDataInterfaceVectorCurve
  &NiagaraDataInterfaceArrayInt32
  &NiagaraDataInterfaceArrayPosition
  &NiagaraDataInterfaceSkeletalMesh
 ]
 usedDynamicInputs []
 usedModules [
  &self:SystemSpawnScript.SystemScriptSource.SystemScriptGraph.NiagaraNodeAssignment_1.SetVariables_#1
  &self:SystemSpawnScript.SystemScriptSource.SystemScriptGraph.NiagaraNodeAssignment_0.SetVariables_#2
  &SystemState
 ]
}
emitters {
 InnerCross {
  emitterUpdateScript [
   &EmitterState()
   &SpawnBurst_Instantaneous(Spawn Time=1, Use Spawn Probability=false, UseLoopCountLimit=false)
  ]
  particleSpawnScript [
   &InitializeParticle(
    Color Mode=::NewEnumerator0
    Lifetime=0.37
    Sprite Size Mode=::NewEnumerator3
    Uniform Sprite Size=14.4566
    UsePositionOffset=false
    Write Material Random=true
   )
   &SpriteFacingAndAlignment(Sprite Alignment=<0 0 1>, Sprite Facing=<0 1 0>, WriteAlignment=true, WriteFacing=true)
   /SetVariables(Particles.Square Glow Scale:Vector3f=&Multiply_VectorByFloat(Vector=<1 1 1>, Float=0.23)) {isSetParameters=true}
   &CameraOffset(Camera Offset Amount=15)
  ]
  particleUpdateScript [
   &ParticleState()
   &SkeletalMeshLocation(Bone / Socket Index=5, Bone Sampling Mode=::NewEnumerator4, MaskSpawnGroup=false, Skeletal Mesh=$System.Skeletal Mesh Reader)
   &ScaleColor(
    Scale Alpha=&FloatFromCurve(FloatCurve=DataInterface() {propertyValues {Curve {keys [<at 0 1 RCIM_Constant>,<at 0.6706 1 RCIM_Constant>,<at 0.7384 0 RCIM_Constant>,<at 0.7968 1 RCIM_Constant>,<at 0.8596 0 RCIM_Constant>,<at 0.9311 1 RCIM_Constant>,<at 0.9897 0 RCIM_Constant>]}}})
    Scale RGB=&MakeVectorFromLinearColorRGB(Color=&ColorFromCurve(
      ColorCurve=DataInterface() {propertyValues {
        AlphaCurve {keys [<at 0 1>]}
        BlueCurve {keys [<at 2.4667 0>,<at 2.5142 0>]}
        GreenCurve {keys [<at 2.4667 0.7083>,<at 2.5142 0>]}
        RedCurve {keys [<at 2.4667 0.1278>,<at 2.5142 1>]}
       }}
      CurveIndex=$System.Age
     ))
    ScaleA=true
    ScaleRGB=true
   )
   /SetVariables(Particles.Square Glow Scale:Vector3f=&Multiply_Vector(
     A=$Particles.Initial.Square Glow Scale
     B=&VectorFromCurve(VectorCurve=DataInterface() {propertyValues {
        XCurve {keys [<at 0 1>,<at 1 1>]}
        YCurve {
         defaultValue=1
         keys [
          <at 0 0 RCIM_Cubic>
          {
           arriveTangent=12.5697
           arriveTangentWeight=0.0333
           interpMode=RCIM_Cubic
           leaveTangent=12.5697
           leaveTangentWeight=0.0624
           tangentMode=RCTM_User
           time=0.1
           value=1
          }
          <at 0.2872 1 RCIM_Cubic>
         ]
        }
        ZCurve {
         defaultValue=0
         keys [
          <at 0 0 RCIM_Cubic>
          {arriveTangent=12.4571, arriveTangentWeight=0.0333, interpMode=RCIM_Cubic, leaveTangent=12.4571, tangentMode=RCTM_User, time=0.1, value=1}
          <at 0.2903 1 RCIM_Cubic>
         ]
        }
       }})
    )) {isSetParameters=true, module=SetVariables_#3}
   &DynamicMaterialParameters(
    Index 0 Param 1=0.03
    Module.Index0Float0Write=true
    Module.Index0Float1Write=true
    Module.Index0Float2Write=true
    Module.Index0Float3Write=true
   )
  ]
  rendererClasses [&NiagaraMeshRendererProperties]
  renderers [{rendererClass=&NiagaraMeshRendererProperties, rendererIndex=0}]
 }
 Lights {
  emitterUpdateScript [
   &EmitterState()
   &SpawnBurst_Instantaneous(Spawn Group=1, Spawn Time=2, Use Spawn Probability=false, UseLoopCountLimit=false) {module=SpawnBurst_Instantaneous001}
   &SpawnBurst_Instantaneous(Use Spawn Probability=false, UseLoopCountLimit=false)
  ]
  particleSpawnScript [&InitializeParticle(
    Lifetime=2
    Position Offset=&MergeHelper_VectorSwitch(0=<0 30 35>, 1=<0 30 100>, 2=?, 3=?, Index=&SpawnGroup(), Num=::NewEnumerator0)
    UsePositionOffset=true
    Write Material Random=true
   )]
  particleUpdateScript [
   &ParticleState()
   &ScaleColor(
    Color Scale:LinearColor=&MergeHelper_ColorSwitch(
     0=&ColorFromCurve(ColorCurve=DataInterface() {propertyValues {
        AlphaCurve {keys [<at 0 0>,<at 0.5 1 RCIM_Cubic>,<at 1 0 RCIM_Cubic>]}
        BlueCurve {keys [<at 0 0.0369 RCIM_Cubic>]}
        GreenCurve {keys [<at 0 3 RCIM_Cubic>]}
        RedCurve {keys [<at 0 0.4412 RCIM_Cubic>]}
       }})
     1=&ColorFromCurve(
      ColorCurve=DataInterface() {propertyValues {
        AlphaCurve {keys [
          {arriveTangent=-0.0047, interpMode=RCIM_Cubic, leaveTangent=-0.0047, tangentMode=RCTM_User, time=0, value=0}
          {arriveTangent=0.4308, interpMode=RCIM_Cubic, leaveTangent=0.4308, time=1.0338, value=1}
          {arriveTangent=1.439, interpMode=RCIM_Cubic, leaveTangent=1.439, tangentMode=RCTM_User, time=2.3212, value=1}
         ]}
        BlueCurve {keys [<at 2.4667 0>,<at 2.5142 0>]}
        GreenCurve {keys [<at 2.4667 3>,<at 2.5142 0>]}
        RedCurve {keys [<at 2.4667 0.1278>,<at 2.5142 3>]}
       }}
      CurveIndex=$System.Age
     )
     2=?
     3=?
     Index=&SpawnGroup()
     Num=::NewEnumerator0
    )
    Scale Mode=3
   )
   &Light_Attributes(
    Light Radius:NiagaraFloat=&MergeHelper_FloatSwitch(0=50, 1=100, 2=?, 3=?, Index=&SpawnGroup(), Num=::NewEnumerator0)
    Radius=1
   )
  ]
  rendererClasses [&NiagaraLightRendererProperties]
  renderers [{rendererClass=&NiagaraLightRendererProperties, rendererIndex=0}]
 }
 OuterFinalSquare {
  emitterUpdateScript [
   &EmitterState()
   &SpawnBurst_Instantaneous(Spawn Time=1.8, Use Spawn Probability=false, UseLoopCountLimit=false)
  ]
  particleSpawnScript [
   &InitializeParticle(
    Color Mode=::NewEnumerator0
    Lifetime=2.06
    Sprite Size Mode=::NewEnumerator3
    Uniform Sprite Size=25
    UsePositionOffset=false
    Write Material Random=true
   )
   &SpriteFacingAndAlignment(Sprite Alignment=<0 0 1>, Sprite Facing=<0 1 0>, WriteAlignment=true, WriteFacing=true)
   /SetVariables(Particles.Square Glow Size:Vector2f=&Vector2DFromFloat(Value=46.795)) {isSetParameters=true}
   /SetVariables(Particles.Square Glow Scale:Vector3f=&Multiply_VectorByFloat(Vector=<1 1 1>, Float=0.6193)) {isSetParameters=true}
   &CameraOffset(Camera Offset Amount=15)
  ]
  particleUpdateScript [
   &ParticleState()
   &SkeletalMeshLocation(Bone / Socket Index=9, Bone Sampling Mode=::NewEnumerator4, MaskSpawnGroup=false, Skeletal Mesh=$System.Skeletal Mesh Reader)
   &ScaleColor(
    Scale Alpha=&FloatFromCurve(
     FloatCurve=DataInterface() {propertyValues {Curve {keys [<at 0 1 RCIM_Constant>,<at 2.6 0 RCIM_Constant>,<at 2.7 1 RCIM_Constant>,<at 2.8 0 RCIM_Constant>,<at 2.9 1 RCIM_Constant>,<at 3 0 RCIM_Constant>,<at 3.1 1 RCIM_Constant>,<at 3.2 0 RCIM_Constant>,<at 3.3 1 RCIM_Constant>,<at 3.4 0 RCIM_Constant>,<at 3.5 1 RCIM_Constant>,<at 3.6 0 RCIM_Constant>,<at 3.7 1 RCIM_Constant>,<at 3.8 0 RCIM_Constant>,<at 3.9 1 RCIM_Constant>]}}}
     CurveIndex=$System.Age
    )
    Scale RGB=&MakeVectorFromLinearColorRGB(Color=&ColorFromCurve(
      ColorCurve=DataInterface() {propertyValues {
        AlphaCurve {keys [<at 0 1>]}
        BlueCurve {keys [<at 2.4667 0>,<at 2.5142 0>]}
        GreenCurve {keys [<at 2.4667 0.7083>,<at 2.5142 0>]}
        RedCurve {keys [<at 2.4667 0.1278>,<at 2.5142 1>]}
       }}
      CurveIndex=$System.Age
     ))
    ScaleA=true
    ScaleRGB=true
   )
   &ScaleSpriteSize(
    Scale Factor=&Vector2DFromCurve(
     Vector2Curve=DataInterface() {propertyValues {
       XCurve {keys [
         <at 0 0.8 RCIM_Cubic>
         {arriveTangent=12.5697, interpMode=RCIM_Cubic, leaveTangent=12.5697, tangentMode=RCTM_User, time=0.0212, value=1}
         {arriveTangent=1.3786, interpMode=RCIM_Cubic, leaveTangent=1.3786, tangentMode=RCTM_User, time=0.0829, value=1}
         {
          arriveTangent=-21.3065
          arriveTangentWeight=0.0488
          interpMode=RCIM_Cubic
          leaveTangentWeight=0.3915
          tangentMode=RCTM_Break
          time=0.1735
          value=1.4
         }
         {arriveTangentWeight=0.1575, interpMode=RCIM_Cubic, leaveTangentWeight=0.145, tangentMode=RCTM_User, time=0.26, value=1.4}
         <at 0.33 0 RCIM_Cubic>
         {arriveTangentWeight=0.0447, interpMode=RCIM_Cubic, leaveTangentWeight=0.3123, tangentMode=RCTM_User, time=0.37, value=1.4}
         {arriveTangentWeight=0.0445, interpMode=RCIM_Cubic, leaveTangentWeight=0.1351, tangentMode=RCTM_User, time=0.5, value=0.95}
         {arriveTangentWeight=0.2652, interpMode=RCIM_Cubic, leaveTangentWeight=0.0112, tangentMode=RCTM_User, time=0.8705, value=0.95}
         {arriveTangentWeight=0.391, interpMode=RCIM_Cubic, leaveTangentWeight=0.0508, tangentMode=RCTM_User, time=0.9, value=0.85}
         <at 1 0 RCIM_Cubic>
        ]}
       YCurve {keys [
         <at 0 0.8 RCIM_Cubic>
         {arriveTangent=12.4571, interpMode=RCIM_Cubic, leaveTangent=12.4571, tangentMode=RCTM_User, time=0.0212, value=1}
         {arriveTangent=2.6261, interpMode=RCIM_Cubic, leaveTangent=2.6261, time=0.0829, value=1}
         {arriveTangent=-22.0114, interpMode=RCIM_Cubic, leaveTangent=0.1224, tangentMode=RCTM_Break, time=0.1735, value=1.4}
         {arriveTangentWeight=0.1352, interpMode=RCIM_Cubic, leaveTangentWeight=0.1672, tangentMode=RCTM_User, time=0.26, value=1.4}
         <at 0.33 0 RCIM_Cubic>
         {arriveTangentWeight=0.0243, interpMode=RCIM_Cubic, leaveTangentWeight=0.127, tangentMode=RCTM_User, time=0.37, value=1.4}
         {arriveTangentWeight=0.046, interpMode=RCIM_Cubic, leaveTangentWeight=0.1337, tangentMode=RCTM_User, time=0.5, value=0.95}
         {arriveTangentWeight=0.2323, interpMode=RCIM_Cubic, leaveTangentWeight=0.0098, tangentMode=RCTM_User, time=0.8705, value=0.95}
         {arriveTangentWeight=0.3914, interpMode=RCIM_Cubic, leaveTangentWeight=0.0504, tangentMode=RCTM_User, time=0.9, value=0.85}
         <at 1 0 RCIM_Cubic>
        ]}
      }}
     ScaleCurve=<1 1>
    )
    Scale Sprite Size Mode=::NewEnumerator2
   )
   /SetVariables(
    Particles.Square Glow Color:LinearColor=&MakeLinearColorFromVectorAndFloat(
     "Vector (RGB)"=&MakeVectorFromLinearColorRGB(Color=&ColorFromCurve(
       ColorCurve=DataInterface() {propertyValues {
         AlphaCurve {keys [<at 0 1>]}
         BlueCurve {keys [<at 2.4667 0>,<at 2.5142 0>]}
         GreenCurve {keys [<at 2.4667 0.7083>,<at 2.5142 0>]}
         RedCurve {keys [<at 2.4667 0.1278>,<at 2.5142 1>]}
        }}
       CurveIndex=$System.Age
      ))
     "Float (Alpha)"=&FloatFromCurve(
      FloatCurve=DataInterface() {propertyValues {Curve {keys [<at 0 1 RCIM_Constant>,<at 2.6 0 RCIM_Constant>,<at 2.7 1 RCIM_Constant>,<at 2.8 0 RCIM_Constant>,<at 2.9 1 RCIM_Constant>,<at 3 0 RCIM_Constant>,<at 3.1 1 RCIM_Constant>,<at 3.2 0 RCIM_Constant>,<at 3.3 1 RCIM_Constant>,<at 3.4 0 RCIM_Constant>,<at 3.5 1 RCIM_Constant>,<at 3.6 0 RCIM_Constant>,<at 3.7 1 RCIM_Constant>,<at 3.8 0 RCIM_Constant>,<at 3.9 1 RCIM_Constant>]}}}
      CurveIndex=$System.Age
     )
    )
    Particles.Square Glow Scale:Vector3f=&Multiply_Vector(
     A=$Particles.Initial.Square Glow Scale
     B=&VectorFromCurve(VectorCurve=DataInterface() {propertyValues {
        XCurve {keys [<at 0 1>,<at 1 1>]}
        YCurve {
         defaultValue=0
         keys [
          <at 0 0.8 RCIM_Cubic>
          {arriveTangent=12.5697, arriveTangentWeight=0.0071, interpMode=RCIM_Cubic, leaveTangent=12.5697, tangentMode=RCTM_User, time=0.0212, value=1}
          <at 0.0829 1 RCIM_Cubic>
          {arriveTangentWeight=0.059, interpMode=RCIM_Cubic, tangentMode=RCTM_User, time=0.26, value=1}
          <at 0.33 0 RCIM_Cubic>
          {arriveTangentWeight=0.0133, interpMode=RCIM_Cubic, tangentMode=RCTM_User, time=0.37, value=1}
          {arriveTangentWeight=0.0433, interpMode=RCIM_Cubic, tangentMode=RCTM_User, time=0.5, value=0.7}
          {arriveTangentWeight=0.1333, interpMode=RCIM_Cubic, tangentMode=RCTM_User, time=0.9, value=0.7}
          <at 1 0 RCIM_Cubic>
         ]
        }
        ZCurve {
         defaultValue=0
         keys [
          <at 0 0.8 RCIM_Cubic>
          {arriveTangent=12.4571, arriveTangentWeight=0.0071, interpMode=RCIM_Cubic, leaveTangent=12.4571, tangentMode=RCTM_User, time=0.0212, value=1}
          <at 0.0829 1 RCIM_Cubic>
          {arriveTangentWeight=0.059, interpMode=RCIM_Cubic, tangentMode=RCTM_User, time=0.26, value=1}
          <at 0.33 0 RCIM_Cubic>
          {arriveTangentWeight=0.0133, interpMode=RCIM_Cubic, tangentMode=RCTM_User, time=0.37, value=1}
          {arriveTangentWeight=0.0433, interpMode=RCIM_Cubic, tangentMode=RCTM_User, time=0.5, value=0.7}
          {arriveTangentWeight=0.1333, interpMode=RCIM_Cubic, tangentMode=RCTM_User, time=0.9, value=0.7}
          <at 1 0 RCIM_Cubic>
         ]
        }
       }})
    )
    Particles.Square Glow Size:Vector2f=&Multiply_Vector2D(
     A=$Particles.Initial.Square Glow Size
     B=&Vector2DFromCurve(
      Vector2Curve=DataInterface() {propertyValues {
        XCurve {keys [
          <at 0 0.8 RCIM_Cubic>
          {arriveTangent=12.5697, interpMode=RCIM_Cubic, leaveTangent=12.5697, tangentMode=RCTM_User, time=0.0212, value=1}
          {arriveTangent=1.3786, interpMode=RCIM_Cubic, leaveTangent=1.3786, tangentMode=RCTM_User, time=0.0829, value=1}
          {
           arriveTangent=-21.3065
           arriveTangentWeight=0.0488
           interpMode=RCIM_Cubic
           leaveTangentWeight=0.3915
           tangentMode=RCTM_Break
           time=0.1735
           value=1.4
          }
          {arriveTangentWeight=0.1575, interpMode=RCIM_Cubic, leaveTangentWeight=0.145, tangentMode=RCTM_User, time=0.26, value=1.4}
          <at 0.33 0 RCIM_Cubic>
          {arriveTangentWeight=0.0447, interpMode=RCIM_Cubic, leaveTangentWeight=0.3123, tangentMode=RCTM_User, time=0.37, value=1.4}
          {arriveTangentWeight=0.0445, interpMode=RCIM_Cubic, leaveTangentWeight=0.1351, tangentMode=RCTM_User, time=0.5, value=0.95}
          {arriveTangentWeight=0.2652, interpMode=RCIM_Cubic, leaveTangentWeight=0.0112, tangentMode=RCTM_User, time=0.8705, value=0.95}
          {arriveTangentWeight=0.391, interpMode=RCIM_Cubic, leaveTangentWeight=0.0508, tangentMode=RCTM_User, time=0.9, value=0.85}
          <at 1 0 RCIM_Cubic>
         ]}
        YCurve {keys [
          <at 0 0.8 RCIM_Cubic>
          {arriveTangent=12.4571, interpMode=RCIM_Cubic, leaveTangent=12.4571, tangentMode=RCTM_User, time=0.0212, value=1}
          {arriveTangent=2.6261, interpMode=RCIM_Cubic, leaveTangent=2.6261, time=0.0829, value=1}
          {arriveTangent=-22.0114, interpMode=RCIM_Cubic, leaveTangent=0.1224, tangentMode=RCTM_Break, time=0.1735, value=1.4}
          {arriveTangentWeight=0.1352, interpMode=RCIM_Cubic, leaveTangentWeight=0.1672, tangentMode=RCTM_User, time=0.26, value=1.4}
          <at 0.33 0 RCIM_Cubic>
          {arriveTangentWeight=0.0243, interpMode=RCIM_Cubic, leaveTangentWeight=0.127, tangentMode=RCTM_User, time=0.37, value=1.4}
          {arriveTangentWeight=0.046, interpMode=RCIM_Cubic, leaveTangentWeight=0.1337, tangentMode=RCTM_User, time=0.5, value=0.95}
          {arriveTangentWeight=0.2323, interpMode=RCIM_Cubic, leaveTangentWeight=0.0098, tangentMode=RCTM_User, time=0.8705, value=0.95}
          {arriveTangentWeight=0.3914, interpMode=RCIM_Cubic, leaveTangentWeight=0.0504, tangentMode=RCTM_User, time=0.9, value=0.85}
          <at 1 0 RCIM_Cubic>
         ]}
       }}
      ScaleCurve=<1 1>
     )
    )
   ) {isSetParameters=true, module=SetVariables_#3}
   &DynamicMaterialParameters(
    Index 0 Param 1=0.03
    Module.Index0Float0Write=true
    Module.Index0Float1Write=true
    Module.Index0Float2Write=true
    Module.Index0Float3Write=true
   )
  ]
  rendererClasses [&NiagaraSpriteRendererProperties, &NiagaraMeshRendererProperties]
  renderers [
   {rendererClass=&NiagaraSpriteRendererProperties}
   {rendererClass=&NiagaraMeshRendererProperties, rendererIndex=1}
  ]
 }
 Spline_Position {
  emitterUpdateScript [
   &EmitterState(
    Inactive Response=::NewEnumerator1
    Life Cycle Mode=::NewEnumerator1
    Loop Behavior=::NewEnumerator1
    Loop Delay=0.4
    Loop Duration=8.75
    UseLoopDelay=true
   )
   &SpawnBurst_Instantaneous(Spawn Count=7, Use Spawn Probability=false, UseLoopCountLimit=false)
  ]
  particleSpawnScript [
   &FN_SetSpawnGroup(Number of Spawn Groups=7)
   &InitializeParticle(
    Color=&MaskLinearColorBySpawnGroup(
     Masked Linear Color=<rgba 0.8594 4.8025 5 1>
     Passthrough Linear Color=&RandomRangeLinearColor(Minimum=<rgba 0.3945 0.6791 0.7 1>, Maximum=<rgba 0.3438 1.921 2 1>)
     Spawn Group Mask=11
    )
    Lifetime Max=8.75
    Lifetime Min=8.5
    Lifetime Mode=::NewEnumerator1
    Position Offset=<0 0 40>
    Sprite Rotation Mode=::NewEnumerator2
    Sprite Size Mode=::NewEnumerator3
    Uniform Sprite Size=&MaskFloatBySpawnGroup(Masked Float=42, Passthrough Float=&RandomRangeFloat(Minimum=24, Maximum=48), Spawn Group Mask=11)
    UsePositionOffset=false
    Write Material Random=true
   )
   /SetVariables(Particles.InitialIndex:NiagaraInt32=$Particles.SpawnGroup) {isSetParameters=true}
   &CameraOffset(Camera Offset Amount=25)
  ]
  particleUpdateScript [
   &ParticleState()
   &ScaleSpriteSize(Uniform Curve Sprite Scale=DataInterface() {propertyValues {Curve {keys [
        <at 0 0 RCIM_Cubic>
        {arriveTangent=10, interpMode=RCIM_Cubic, leaveTangent=10, time=0.056, value=1.2}
        {arriveTangent=0.2647, interpMode=RCIM_Cubic, leaveTangent=0.2647, tangentMode=RCTM_User, time=0.1, value=1}
        <at 1 1>
       ]}}})
   &SkeletalMeshLocation(
    Bone / Socket Index=&Add_Integer(A=3, B=$Particles.InitialIndex)
    Bone Sampling Mode=::NewEnumerator4
    MaskSpawnGroup=false
    Position Sampling=::NewEnumerator2
    Sampled Position Offset=&MergeHelper_VectorSwitch(
     0=<-7 0 7>
     1=<7 0 -7>
     2=<-14 0 -14>
     3=<10 0 -10>
     4=<-6 0 -5>
     5=<-7.5 0 -7.5>
     6=<13 0 -13>
     Index=$Particles.InitialIndex
     Num=::NewEnumerator5
    )
    Skeletal Mesh=$System.Skeletal Mesh Reader
   )
   &SkeletalMeshLocation(
    Bone / Socket Index=&Add_Integer(A=3, B=$Particles.InitialIndex)
    Bone Sampling Mode=::NewEnumerator4
    MaskSpawnGroup=false
    Position Sampling=::NewEnumerator2
    Sampled Position Offset=&MergeHelper_VectorSwitch(
     0=<-8 0 8>
     1=<7 0 7>
     2=<-14 0 14>
     3=<8 0 10>
     4=<4 0 -4>
     5=<-8 0 -7.6>
     6=<13 0 -13>
     Index=$Particles.InitialIndex
     Num=::NewEnumerator5
    )
    Skeletal Mesh=$System.Skeletal Mesh Reader
   ) {module=SkeletalMeshLocation001}
   &self:ScratchModule(
    Ending Position:NiagaraPosition=$Particles.SkeletalMeshLocation001.SampledPosition
    Ending Position Array:NiagaraDataInterfaceArrayPosition=$System.Ending Position Array
    FlareIndex:NiagaraFloat=&MakeFloatFromInt(Int=$Particles.InitialIndex)
    Starting Position:NiagaraPosition=$Particles.SkeletalMeshLocation.SampledPosition
    Starting Position Array:NiagaraDataInterfaceArrayPosition=$System.Starting Position Array
   ) {module=ScratchModule}
  ]
  rendererClasses [&NiagaraSpriteRendererProperties]
  renderers [{rendererClass=&NiagaraSpriteRendererProperties}]
 }
 Splines {
  emitterUpdateScript [
   &EmitterState(
    Inactive Response=::NewEnumerator1
    Life Cycle Mode=::NewEnumerator1
    Loop Behavior=::NewEnumerator1
    Loop Duration=4.7333
    UseLoopDelay=true
   )
   &SpawnBurst_Instantaneous(Spawn Time=0.66, Use Spawn Probability=false, UseLoopCountLimit=false)
   &SpawnBurst_Instantaneous(Spawn Group=1, Spawn Time=0.867, Use Spawn Probability=false, UseLoopCountLimit=false) {module=SpawnBurst_Instantaneous001}
   &SpawnBurst_Instantaneous(Spawn Group=2, Spawn Time=1.12, Use Spawn Probability=false, UseLoopCountLimit=false) {module=SpawnBurst_Instantaneous002}
   &SpawnBurst_Instantaneous(Spawn Group=3, Spawn Time=1.33, Use Spawn Probability=false, UseLoopCountLimit=false) {module=SpawnBurst_Instantaneous003}
   &SpawnBurst_Instantaneous(Spawn Group=4, Spawn Time=1.43, Use Spawn Probability=false, UseLoopCountLimit=false) {module=SpawnBurst_Instantaneous004}
   &SpawnBurst_Instantaneous(Spawn Group=5, Spawn Time=1.73, Use Spawn Probability=false, UseLoopCountLimit=false) {module=SpawnBurst_Instantaneous005}
   &SpawnBurst_Instantaneous(Spawn Group=6, Spawn Time=1.2, Use Spawn Probability=false, UseLoopCountLimit=false) {module=SpawnBurst_Instantaneous006}
   &SpawnBurst_Instantaneous(Spawn Time=4.7333, Use Spawn Probability=false, UseLoopCountLimit=false) {module=SpawnBurst_Instantaneous007}
  ]
  particleSpawnScript [
   &FN_SetSpawnGroup(Number of Spawn Groups=13)
   /SetVariables(
    Particles.IndexA:NiagaraInt32=&SelectIntFromArray(Direct Array Index=$Emitter.SpawnGroup, Int Selection Array=DataInterface() {propertyValues {IntData [0,1,2,3,4,5,6]}})
    Particles.IndexB:NiagaraInt32=&SelectIntFromArray(Direct Array Index=$Emitter.SpawnGroup, Int Selection Array=DataInterface() {propertyValues {IntData [1,2,3,4,5,6,7]}})
   ) {isSetParameters=true}
   &InitializeParticle(
    Color=<rgba 0.1278 0.7083 0 1>
    Lifetime=&MergeHelper_FloatSwitch(0=0.2, 1=0.2, 2=0.2, 3=0.2, 4=0.2, 5=0.2, 6=0.23, Index=$Particles.IndexA, Num=::NewEnumerator5)
    Mesh Scale Mode=::NewEnumerator3
    Position Offset=<0 0 -80>
    UsePositionOffset=true
    Write Material Random=true
   )
   /SetVariables(Particles.InitialIndex:NiagaraInt32=$Emitter.SpawnGroup) {enabled=false, isSetParameters=true}
   &CameraOffset(Camera Offset Amount=15)
  ]
  particleUpdateScript [
   &ParticleState()
   &self:ScratchModule_01(
    Ending Position Array:NiagaraDataInterfaceArrayPosition=$System.Ending Position Array
    Index:NiagaraInt32=$Particles.IndexA
    Index02:NiagaraInt32=$Particles.IndexB
    Position Array:NiagaraDataInterfaceArrayPosition=$System.Starting Position Array
   ) {module=ScratchModule_01}
   &ScaleMeshSize(Scale Factor=&MakeVector(X=&Divide_Float(A=$Particles.LightningLength, B=100), Y=1, Z=1))
   &DynamicMaterialParameters(
    Index 0 Param 1=0.344
    Index 0 Param 2=&FloatFromCurve(FloatCurve=DataInterface() {propertyValues {Curve {keys [<at 0 -1>,<at 1 1>]}}})
    Index 0 Param 3=&MergeHelper_FloatSwitch(0=4, 1=4, 2=5, 3=3, 4=3, 5=4, 6=4, Index=$Particles.IndexA, Num=::NewEnumerator5)
    Module.Index0Float0Write=true
    Module.Index0Float1Write=true
    Module.Index0Float2Write=true
    Module.Index0Float3Write=false
   )
   &DynamicMaterialParameters(
    Index 0 Param 1=15
    Index 0 Param 2=1
    Index 0 Param 3=0.1
    Module.Index0Float0Write=true
    Module.Index0Float1Write=true
    Module.Index0Float2Write=true
    Module.Index0Float3Write=true
   ) {enabled=false, module=DynamicMaterialParameters001}
   &ScaleColor(
    Scale RGB=&VectorFromFloat(Value=&FloatFromCurve(
      FloatCurve=DataInterface() {propertyValues {Curve {keys [
          <at 0 1>
          <at 1.25 1>
          <at 1.35 8>
          <at 1.45 1>
          <at 2.6 1>
          <at 2.7 8>
          <at 2.85 1>
          {arriveTangent=0.139, interpMode=RCIM_Cubic, leaveTangent=0.139, tangentMode=RCTM_User, time=3.4, value=1.0195}
          {arriveTangent=12.2005, interpMode=RCIM_Cubic, leaveTangent=12.2005, time=4.07, value=8}
          {arriveTangent=-15.1515, interpMode=RCIM_Cubic, leaveTangent=-15.1515, time=4.3, value=12}
          <at 4.4 3>
         ]}}}
      CurveIndex=$Emitter.LoopedAge
     ))
    ScaleA=true
    ScaleRGB=false
   )
  ]
  rendererClasses [&NiagaraMeshRendererProperties]
  renderers [{rendererClass=&NiagaraMeshRendererProperties, rendererIndex=0}]
 }
 Squares {
  emitterUpdateScript [
   &EmitterState()
   &SpawnBurst_Instantaneous(Spawn Group=9, Spawn Time=1.8, Use Spawn Probability=false, UseLoopCountLimit=false) {module=SpawnBurst_Instantaneous009}
   &SpawnBurst_Instantaneous(Spawn Group=8, Spawn Time=1.434, Use Spawn Probability=false, UseLoopCountLimit=false) {module=SpawnBurst_Instantaneous008}
   &SpawnBurst_Instantaneous(Spawn Group=7, Spawn Time=1.3, Use Spawn Probability=false, UseLoopCountLimit=false) {module=SpawnBurst_Instantaneous007}
   &SpawnBurst_Instantaneous(Spawn Group=6, Spawn Time=1.2, Use Spawn Probability=false, UseLoopCountLimit=false) {module=SpawnBurst_Instantaneous006}
   &SpawnBurst_Instantaneous(Spawn Group=5, Spawn Time=0.934, Use Spawn Probability=false, UseLoopCountLimit=false) {module=SpawnBurst_Instantaneous005}
   &SpawnBurst_Instantaneous(Spawn Group=4, Spawn Time=0.834, Use Spawn Probability=false, UseLoopCountLimit=false) {module=SpawnBurst_Instantaneous004}
   &SpawnBurst_Instantaneous(Spawn Group=3, Spawn Time=0.534, Use Spawn Probability=false, UseLoopCountLimit=false) {module=SpawnBurst_Instantaneous003}
   &SpawnBurst_Instantaneous(Spawn Group=2, Spawn Time=0.466, Use Spawn Probability=false, UseLoopCountLimit=false) {module=SpawnBurst_Instantaneous002}
   &SpawnBurst_Instantaneous(Spawn Group=1, Spawn Time=0.166, Use Spawn Probability=false, UseLoopCountLimit=false) {module=SpawnBurst_Instantaneous001}
   &SpawnBurst_Instantaneous(Use Spawn Probability=false, UseLoopCountLimit=false)
  ]
  particleSpawnScript [
   &InitializeParticle(
    Color Mode=::NewEnumerator0
    Lifetime=&MergeHelper_FloatSwitch(
     0=0.2
     1=0.133
     2=0.133
     3=0.33
     4=0.33
     5=0.37
     6=0.467
     7=0.35
     8=0.5
     9=2.06
     Index=&SpawnGroup()
     Num=::NewEnumerator8
    )
    Sprite Size Mode=::NewEnumerator3
    Uniform Sprite Size=&MergeHelper_FloatSwitch(0=12, 1=7, 2=9, 3=18, 4=18, 5=30, 6=20, 7=10, 8=15, 9=25, Index=&SpawnGroup(), Num=::NewEnumerator8)
    UsePositionOffset=false
    Write Material Random=true
   )
   &SpriteFacingAndAlignment(Sprite Alignment=<0 0 1>, Sprite Facing=<0 1 0>, WriteAlignment=true, WriteFacing=true)
   &CameraOffset(Camera Offset Amount=15)
   /SetVariables(Particles.Square Glow Scale:Vector3f=&Multiply_VectorByFloat(
     Vector=<1 1 1>
     Float=&MergeHelper_FloatSwitch(
      0=0.2
      1=0.12
      2=0.15
      3=0.33
      4=0.3
      5=0.53
      6=0.34
      7=0.16
      8=0.28
      9=0.49
      Index=&SpawnGroup()
      Num=::NewEnumerator8
     )
    )) {isSetParameters=true}
  ]
  particleUpdateScript [
   &ParticleState()
   &SkeletalMeshLocation(
    Bone / Socket Index=&SelectIntFromArray(
     Direct Array Index=&SpawnGroup()
     Int Selection Array=DataInterface() {propertyValues {IntData [0,1,2,3,4,5,6,7,8,9]}}
    )
    Bone Sampling Mode=::NewEnumerator4
    MaskSpawnGroup=false
    Skeletal Mesh=$System.Skeletal Mesh Reader
   )
   &ScaleSpriteSize(
    Scale Factor=&MergeHelper_Vector2Switch(
     0=&Vector2DFromCurve(
      Vector2Curve=DataInterface() {propertyValues {
        XCurve {keys [
          <at 0 0 RCIM_Cubic>
          {arriveTangent=12.5697, interpMode=RCIM_Cubic, leaveTangent=12.5697, tangentMode=RCTM_User, time=0.1, value=1}
          <at 0.2872 1 RCIM_Cubic>
         ]}
        YCurve {keys [
          <at 0 0 RCIM_Cubic>
          {arriveTangent=12.4571, interpMode=RCIM_Cubic, leaveTangent=12.4571, tangentMode=RCTM_User, time=0.1, value=1}
          <at 0.2903 1 RCIM_Cubic>
         ]}
       }}
      ScaleCurve=<1 1>
     )
     1=&Vector2DFromCurve(
      Vector2Curve=DataInterface() {propertyValues {
        XCurve {keys [
          <at 0 0 RCIM_Cubic>
          {arriveTangent=12.5697, interpMode=RCIM_Cubic, leaveTangent=12.5697, tangentMode=RCTM_User, time=0.1, value=1}
          <at 0.2872 1 RCIM_Cubic>
         ]}
        YCurve {keys [
          <at 0 0 RCIM_Cubic>
          {arriveTangent=12.4571, interpMode=RCIM_Cubic, leaveTangent=12.4571, tangentMode=RCTM_User, time=0.1, value=1}
          <at 0.2903 1 RCIM_Cubic>
         ]}
       }}
      ScaleCurve=<1 1>
     )
     2=&Vector2DFromCurve(
      Vector2Curve=DataInterface() {propertyValues {
        XCurve {keys [
          <at 0 0 RCIM_Cubic>
          {arriveTangent=12.5697, interpMode=RCIM_Cubic, leaveTangent=12.5697, tangentMode=RCTM_User, time=0.1, value=1}
          <at 0.2872 1 RCIM_Cubic>
         ]}
        YCurve {keys [
          <at 0 0 RCIM_Cubic>
          {arriveTangent=12.4571, interpMode=RCIM_Cubic, leaveTangent=12.4571, tangentMode=RCTM_User, time=0.1, value=1}
          <at 0.2903 1 RCIM_Cubic>
         ]}
       }}
      ScaleCurve=<1 1>
     )
     3=&Vector2DFromCurve(
      Vector2Curve=DataInterface() {propertyValues {
        XCurve {keys [
          <at 0 0.8 RCIM_Cubic>
          {
           arriveTangent=5.0121
           arriveTangentWeight=0.2282
           interpMode=RCIM_Cubic
           leaveTangent=5.0121
           leaveTangentWeight=0.1124
           tangentMode=RCTM_User
           time=0.2
           value=1
          }
          <at 0.2985 1 RCIM_Cubic>
          {arriveTangentWeight=0.2344, interpMode=RCIM_Cubic, leaveTangentWeight=0.1728, tangentMode=RCTM_User, time=0.7, value=1}
          <at 1 0 RCIM_Cubic>
         ]}
        YCurve {keys [
          <at 0 0.1 RCIM_Cubic>
          {
           arriveTangent=5.0121
           arriveTangentWeight=0.3622
           interpMode=RCIM_Cubic
           leaveTangent=5.0121
           leaveTangentWeight=0.1786
           tangentMode=RCTM_User
           time=0.2
           value=1
          }
          <at 0.2986 1 RCIM_Cubic>
          {arriveTangentWeight=0.233, interpMode=RCIM_Cubic, leaveTangentWeight=0.1741, tangentMode=RCTM_User, time=0.7, value=1}
          <at 1 0 RCIM_Cubic>
         ]}
       }}
      ScaleCurve=<1 1>
     )
     4=&Vector2DFromCurve(
      Vector2Curve=DataInterface() {propertyValues {
        XCurve {keys [
          <at 0 0.2 RCIM_Cubic>
          {arriveTangent=4.0016, interpMode=RCIM_Cubic, leaveTangent=4.0016, time=0.1, value=1}
          <at 0.1999 1 RCIM_Cubic>
          {arriveTangentWeight=0.2668, interpMode=RCIM_Cubic, leaveTangentWeight=0.1601, tangentMode=RCTM_User, time=0.7, value=1}
          <at 1 0 RCIM_Cubic>
         ]}
        YCurve {keys [
          <at 0 0.8 RCIM_Cubic>
          {arriveTangent=5.7128, interpMode=RCIM_Cubic, leaveTangent=5.7128, tangentMode=RCTM_User, time=0.1, value=1}
          <at 0.1945 1 RCIM_Cubic>
          {arriveTangentWeight=0.2686, interpMode=RCIM_Cubic, leaveTangentWeight=0.1594, tangentMode=RCTM_User, time=0.7, value=1}
          <at 1 0 RCIM_Cubic>
         ]}
       }}
      ScaleCurve=<1 1>
     )
     5=&Vector2DFromCurve(
      Vector2Curve=DataInterface() {propertyValues {
        XCurve {keys [
          <at 0 0.8 RCIM_Cubic>
          {
           arriveTangent=5.0121
           arriveTangentWeight=0.2282
           interpMode=RCIM_Cubic
           leaveTangent=5.0121
           leaveTangentWeight=0.1124
           tangentMode=RCTM_User
           time=0.2
           value=1
          }
          <at 0.2985 1 RCIM_Cubic>
          {arriveTangentWeight=0.2344, interpMode=RCIM_Cubic, leaveTangentWeight=0.1728, tangentMode=RCTM_User, time=0.7, value=1}
          <at 1 1 RCIM_Cubic>
         ]}
        YCurve {keys [
          <at 0 0.1 RCIM_Cubic>
          {
           arriveTangent=5.0121
           arriveTangentWeight=0.3622
           interpMode=RCIM_Cubic
           leaveTangent=5.0121
           leaveTangentWeight=0.1786
           tangentMode=RCTM_User
           time=0.2
           value=1
          }
          <at 0.2986 1 RCIM_Cubic>
          {arriveTangentWeight=0.233, interpMode=RCIM_Cubic, leaveTangentWeight=0.1741, tangentMode=RCTM_User, time=0.7, value=1}
          <at 1 1 RCIM_Cubic>
         ]}
       }}
      ScaleCurve=<1 1>
     )
     6=&Vector2DFromCurve(
      Vector2Curve=DataInterface() {propertyValues {
        XCurve {keys [
          <at 0 0 RCIM_Cubic>
          {arriveTangent=12.5697, interpMode=RCIM_Cubic, leaveTangent=12.5697, tangentMode=RCTM_User, time=0.05, value=1}
          <at 0.15 1 RCIM_Cubic>
          {arriveTangent=-1.1765, interpMode=RCIM_Cubic, leaveTangent=-1.1765, time=0.7, value=1}
          <at 1 0 RCIM_Cubic>
         ]}
        YCurve {keys [
          <at 0 0 RCIM_Cubic>
          {arriveTangent=12.4571, interpMode=RCIM_Cubic, leaveTangent=12.4571, tangentMode=RCTM_User, time=0.05, value=1}
          <at 0.15 1 RCIM_Cubic>
          {arriveTangent=-1.1765, interpMode=RCIM_Cubic, leaveTangent=-1.1765, time=0.7, value=1}
          <at 1 0 RCIM_Cubic>
         ]}
       }}
      ScaleCurve=<1 1>
     )
     7=&Vector2DFromCurve(
      Vector2Curve=DataInterface() {propertyValues {
        XCurve {keys [
          <at 0 0 RCIM_Cubic>
          {arriveTangent=12.5697, interpMode=RCIM_Cubic, leaveTangent=12.5697, tangentMode=RCTM_User, time=0.1, value=1}
          <at 0.2 1 RCIM_Cubic>
          {arriveTangent=-1.25, interpMode=RCIM_Cubic, leaveTangent=-1.25, time=0.7, value=1}
          <at 1 0 RCIM_Cubic>
         ]}
        YCurve {keys [
          <at 0 0 RCIM_Cubic>
          {arriveTangent=12.4571, interpMode=RCIM_Cubic, leaveTangent=12.4571, tangentMode=RCTM_User, time=0.1, value=1}
          <at 0.2 1 RCIM_Cubic>
          {arriveTangent=-1.25, interpMode=RCIM_Cubic, leaveTangent=-1.25, time=0.7, value=1}
          <at 1 0 RCIM_Cubic>
         ]}
       }}
      ScaleCurve=<1 1>
     )
     8=&Vector2DFromCurve(
      Vector2Curve=DataInterface() {propertyValues {
        XCurve {keys [
          <at 0 0 RCIM_Cubic>
          {arriveTangent=12.5697, interpMode=RCIM_Cubic, leaveTangent=12.5697, tangentMode=RCTM_User, time=0.05, value=1}
          {arriveTangent=3.9351, interpMode=RCIM_Cubic, leaveTangent=3.9351, tangentMode=RCTM_User, time=0.1477, value=1.05}
          {arriveTangentWeight=0.0393, interpMode=RCIM_Cubic, leaveTangentWeight=0.1448, tangentMode=RCTM_User, time=0.2656, value=1.05}
          {arriveTangentWeight=0.2526, interpMode=RCIM_Cubic, leaveTangentWeight=0.1745, tangentMode=RCTM_User, time=0.9, value=1.05}
          <at 1 0 RCIM_Cubic>
         ]}
        YCurve {keys [
          <at 0 0 RCIM_Cubic>
          {arriveTangent=12.4571, interpMode=RCIM_Cubic, leaveTangent=12.4571, tangentMode=RCTM_User, time=0.05, value=1}
          {arriveTangent=5.1603, interpMode=RCIM_Cubic, leaveTangent=5.1603, tangentMode=RCTM_User, time=0.1508, value=1.05}
          {arriveTangentWeight=0.039, interpMode=RCIM_Cubic, leaveTangentWeight=0.1441, tangentMode=RCTM_User, time=0.2678, value=1.05}
          {arriveTangentWeight=0.2519, interpMode=RCIM_Cubic, leaveTangentWeight=0.1748, tangentMode=RCTM_User, time=0.9, value=1.05}
          <at 1 0 RCIM_Cubic>
         ]}
       }}
      ScaleCurve=<1 1>
     )
     9=&Vector2DFromCurve(
      Vector2Curve=DataInterface() {propertyValues {
        XCurve {keys [
          <at 0 0.8 RCIM_Cubic>
          {arriveTangent=12.5697, interpMode=RCIM_Cubic, leaveTangent=12.5697, tangentMode=RCTM_User, time=0.0212, value=1}
          <at 0.0829 1 RCIM_Cubic>
          {arriveTangentWeight=0.2461, interpMode=RCIM_Cubic, leaveTangentWeight=0.0973, tangentMode=RCTM_User, time=0.26, value=1}
          <at 0.33 0 RCIM_Cubic>
          {arriveTangentWeight=0.0269, interpMode=RCIM_Cubic, leaveTangentWeight=0.3568, tangentMode=RCTM_User, time=0.37, value=1}
          {arriveTangentWeight=0.0641, interpMode=RCIM_Cubic, leaveTangentWeight=0.1247, tangentMode=RCTM_User, time=0.5, value=0.7}
          {arriveTangentWeight=0.1167, interpMode=RCIM_Cubic, leaveTangentWeight=0.0333, tangentMode=RCTM_User, time=0.9, value=0.7}
          <at 1 0 RCIM_Cubic>
         ]}
        YCurve {keys [
          <at 0 0.8 RCIM_Cubic>
          {arriveTangent=12.4571, interpMode=RCIM_Cubic, leaveTangent=12.4571, tangentMode=RCTM_User, time=0.0212, value=1}
          <at 0.0829 1 RCIM_Cubic>
          {arriveTangentWeight=0.059, interpMode=RCIM_Cubic, leaveTangentWeight=0.0233, tangentMode=RCTM_User, time=0.26, value=1}
          <at 0.33 0 RCIM_Cubic>
          {arriveTangentWeight=0.0133, interpMode=RCIM_Cubic, leaveTangentWeight=0.1767, tangentMode=RCTM_User, time=0.37, value=1}
          {arriveTangentWeight=0.0641, interpMode=RCIM_Cubic, leaveTangentWeight=0.1247, tangentMode=RCTM_User, time=0.5, value=0.7}
          {arriveTangentWeight=0.1167, interpMode=RCIM_Cubic, leaveTangentWeight=0.0333, tangentMode=RCTM_User, time=0.9, value=0.7}
          <at 1 0 RCIM_Cubic>
         ]}
       }}
      ScaleCurve=<1 1>
     )
     Index=&SpawnGroup()
     Num=::NewEnumerator8
    )
    Scale Sprite Size Mode=::NewEnumerator2
   )
   &ScaleColor(
    Scale Alpha=&MergeHelper_FloatSwitch(
     0=1
     1=1
     2=1
     3=1
     4=1
     5=1
     6=&FloatFromCurve(FloatCurve=DataInterface() {propertyValues {Curve {keys [<at 0 1 RCIM_Constant>,<at 0.4988 1 RCIM_Constant>,<at 0.6316 0 RCIM_Constant>,<at 0.7048 1 RCIM_Constant>,<at 0.7899 0 RCIM_Constant>,<at 0.8491 1 RCIM_Constant>,<at 0.974 0 RCIM_Constant>]}}})
     7=&FloatFromCurve(FloatCurve=DataInterface() {propertyValues {Curve {keys [<at 0 1 RCIM_Constant>,<at 0.4988 1 RCIM_Constant>,<at 0.6316 0 RCIM_Constant>,<at 0.7048 1 RCIM_Constant>,<at 0.7899 0 RCIM_Constant>,<at 0.8491 1 RCIM_Constant>,<at 0.974 0 RCIM_Constant>]}}})
     8=1
     9=&FloatFromCurve(
      FloatCurve=DataInterface() {propertyValues {Curve {keys [<at 0 1 RCIM_Constant>,<at 2.6 0 RCIM_Constant>,<at 2.7 1 RCIM_Constant>,<at 2.8 0 RCIM_Constant>,<at 2.9 1 RCIM_Constant>,<at 3 0 RCIM_Constant>,<at 3.1 1 RCIM_Constant>,<at 3.2 0 RCIM_Constant>,<at 3.3 1 RCIM_Constant>,<at 3.4 0 RCIM_Constant>,<at 3.5 1 RCIM_Constant>,<at 3.6 0 RCIM_Constant>,<at 3.7 1 RCIM_Constant>,<at 3.8 0 RCIM_Constant>,<at 3.9 1 RCIM_Constant>]}}}
      CurveIndex=$System.Age
     )
     Index=&SpawnGroup()
     Num=::NewEnumerator8
    )
    Scale RGB=&MakeVectorFromLinearColorRGB(Color=&ColorFromCurve(
      ColorCurve=DataInterface() {propertyValues {
        AlphaCurve {keys [<at 0 1>]}
        BlueCurve {keys [<at 2.4667 0>,<at 2.5142 0>]}
        GreenCurve {keys [<at 2.4667 0.7083>,<at 2.5142 0>]}
        RedCurve {keys [<at 2.4667 0.1278>,<at 2.5142 1>]}
       }}
      CurveIndex=$System.Age
     ))
    ScaleA=true
    ScaleRGB=true
   )
   /SetVariables(
    Particles.Square Glow Color:LinearColor=&MakeLinearColorFromVectorAndFloat(
     "Vector (RGB)"=&MakeVectorFromLinearColorRGB(Color=&ColorFromCurve(
       ColorCurve=DataInterface() {propertyValues {
         AlphaCurve {keys [<at 0 1>]}
         BlueCurve {keys [<at 2.4667 0>,<at 2.5142 0>]}
         GreenCurve {keys [<at 2.4667 0.7083>,<at 2.5142 0>]}
         RedCurve {keys [<at 2.4667 0.1278>,<at 2.5142 1>]}
        }}
       CurveIndex=$System.Age
      ))
     "Float (Alpha)"=&Multiply_Float(
      A=&MergeHelper_FloatSwitch(
       0=1
       1=1
       2=1
       3=1
       4=1
       5=1
       6=&FloatFromCurve(FloatCurve=DataInterface() {propertyValues {Curve {keys [<at 0 1 RCIM_Constant>,<at 0.4988 1 RCIM_Constant>,<at 0.6316 0 RCIM_Constant>,<at 0.7048 1 RCIM_Constant>,<at 0.7899 0 RCIM_Constant>,<at 0.8491 1 RCIM_Constant>,<at 0.974 0 RCIM_Constant>]}}})
       7=&FloatFromCurve(FloatCurve=DataInterface() {propertyValues {Curve {keys [<at 0 1 RCIM_Constant>,<at 0.4988 1 RCIM_Constant>,<at 0.6316 0 RCIM_Constant>,<at 0.7048 1 RCIM_Constant>,<at 0.7899 0 RCIM_Constant>,<at 0.8491 1 RCIM_Constant>,<at 0.974 0 RCIM_Constant>]}}})
       8=1
       9=&FloatFromCurve(
        FloatCurve=DataInterface() {propertyValues {Curve {keys [<at 0 1 RCIM_Constant>,<at 2.6 0 RCIM_Constant>,<at 2.7 1 RCIM_Constant>,<at 2.8 0 RCIM_Constant>,<at 2.9 1 RCIM_Constant>,<at 3 0 RCIM_Constant>,<at 3.1 1 RCIM_Constant>,<at 3.2 0 RCIM_Constant>,<at 3.3 1 RCIM_Constant>,<at 3.4 0 RCIM_Constant>,<at 3.5 1 RCIM_Constant>,<at 3.6 0 RCIM_Constant>,<at 3.7 1 RCIM_Constant>,<at 3.8 0 RCIM_Constant>,<at 3.9 1 RCIM_Constant>]}}}
        CurveIndex=$System.Age
       )
       Index=&SpawnGroup()
       Num=::NewEnumerator8
      )
      B=0.5
     )
    )
    Particles.Square Glow Scale:Vector3f=&MergeHelper_VectorSwitch(
     0=&Multiply_Vector(
      A=$Particles.Initial.Square Glow Scale
      B=&VectorFromCurve(VectorCurve=DataInterface() {propertyValues {
         XCurve {keys [<at 0 1>,<at 1 1>]}
         YCurve {
          defaultValue=1
          keys [
           <at 0 0 RCIM_Cubic>
           {arriveTangent=12.5697, arriveTangentWeight=0.0333, interpMode=RCIM_Cubic, leaveTangent=12.5697, tangentMode=RCTM_User, time=0.1, value=1}
           <at 0.2872 1 RCIM_Cubic>
          ]
         }
         ZCurve {keys [
           <at 0 0 RCIM_Cubic>
           {
            arriveTangent=12.5697
            arriveTangentWeight=0.0333
            interpMode=RCIM_Cubic
            leaveTangent=12.5697
            leaveTangentWeight=0.3
            tangentMode=RCTM_User
            time=0.1
            value=1
           }
           {interpMode=RCIM_Cubic, leaveTangent=-1.1111, time=0.2872, value=1}
          ]}
        }})
     )
     1=&Multiply_Vector(
      A=$Particles.Initial.Square Glow Scale
      B=&VectorFromCurve(VectorCurve=DataInterface() {propertyValues {
         XCurve {keys [<at 0 1>,<at 1 1>]}
         YCurve {
          defaultValue=1
          keys [
           <at 0 0 RCIM_Cubic>
           {
            arriveTangent=12.5697
            arriveTangentWeight=0.0333
            interpMode=RCIM_Cubic
            leaveTangent=12.5697
            leaveTangentWeight=0.0624
            tangentMode=RCTM_User
            time=0.1
            value=1
           }
           <at 0.2872 1 RCIM_Cubic>
          ]
         }
         ZCurve {
          defaultValue=0
          keys [
           <at 0 0 RCIM_Cubic>
           {arriveTangent=12.4571, arriveTangentWeight=0.0333, interpMode=RCIM_Cubic, leaveTangent=12.4571, tangentMode=RCTM_User, time=0.1, value=1}
           <at 0.2903 1 RCIM_Cubic>
          ]
         }
        }})
     )
     2=&Multiply_Vector(
      A=$Particles.Initial.Square Glow Scale
      B=&VectorFromCurve(VectorCurve=DataInterface() {propertyValues {
         XCurve {keys [<at 0 1>,<at 1 1>]}
         YCurve {
          defaultValue=1
          keys [
           <at 0 0 RCIM_Cubic>
           {
            arriveTangent=12.5697
            arriveTangentWeight=0.0333
            interpMode=RCIM_Cubic
            leaveTangent=12.5697
            leaveTangentWeight=0.0624
            tangentMode=RCTM_User
            time=0.1
            value=1
           }
           <at 0.2872 1 RCIM_Cubic>
          ]
         }
         ZCurve {
          defaultValue=0
          keys [
           <at 0 0 RCIM_Cubic>
           {arriveTangent=12.4571, arriveTangentWeight=0.0333, interpMode=RCIM_Cubic, leaveTangent=12.4571, tangentMode=RCTM_User, time=0.1, value=1}
           <at 0.2903 1 RCIM_Cubic>
          ]
         }
        }})
     )
     3=&Multiply_Vector(
      A=$Particles.Initial.Square Glow Scale
      B=&VectorFromCurve(VectorCurve=DataInterface() {propertyValues {
         XCurve {keys [<at 0 1>,<at 1 1>]}
         YCurve {
          defaultValue=0
          keys [
           <at 0 0.8 RCIM_Cubic>
           {arriveTangent=5.0121, arriveTangentWeight=0.0667, interpMode=RCIM_Cubic, leaveTangent=5.0121, tangentMode=RCTM_User, time=0.2, value=1}
           <at 0.2985 1 RCIM_Cubic>
           {arriveTangentWeight=0.1338, interpMode=RCIM_Cubic, tangentMode=RCTM_User, time=0.7, value=1}
           <at 1 0 RCIM_Cubic>
          ]
         }
         ZCurve {
          defaultValue=0
          keys [
           <at 0 0.1 RCIM_Cubic>
           {arriveTangent=5.0121, arriveTangentWeight=0.0667, interpMode=RCIM_Cubic, leaveTangent=5.0121, tangentMode=RCTM_User, time=0.2, value=1}
           <at 0.2986 1 RCIM_Cubic>
           {arriveTangentWeight=0.1338, interpMode=RCIM_Cubic, tangentMode=RCTM_User, time=0.7, value=1}
           <at 1 0 RCIM_Cubic>
          ]
         }
        }})
     )
     4=&Multiply_Vector(
      A=$Particles.Initial.Square Glow Scale
      B=&VectorFromCurve(VectorCurve=DataInterface() {propertyValues {
         XCurve {keys [<at 0 1>,<at 1 1>]}
         YCurve {
          defaultValue=0.8
          keys [
           <at 0 0.2 RCIM_Cubic>
           {arriveTangent=4.0016, interpMode=RCIM_Cubic, leaveTangent=4.0016, time=0.1, value=1}
           <at 0.1999 1 RCIM_Cubic>
           {arriveTangentWeight=0.1667, interpMode=RCIM_Cubic, tangentMode=RCTM_User, time=0.7, value=1}
           <at 1 0 RCIM_Cubic>
          ]
         }
         ZCurve {
          defaultValue=0.1
          keys [
           <at 0 0.8 RCIM_Cubic>
           {arriveTangent=5.7128, arriveTangentWeight=0.0333, interpMode=RCIM_Cubic, leaveTangent=5.7128, tangentMode=RCTM_User, time=0.1, value=1}
           <at 0.1945 1 RCIM_Cubic>
           {arriveTangentWeight=0.1685, interpMode=RCIM_Cubic, tangentMode=RCTM_User, time=0.7, value=1}
           <at 1 0 RCIM_Cubic>
          ]
         }
        }})
     )
     5=&Multiply_Vector(
      A=$Particles.Initial.Square Glow Scale
      B=&VectorFromCurve(VectorCurve=DataInterface() {propertyValues {
         XCurve {keys [<at 0 1>,<at 1 1>]}
         YCurve {
          defaultValue=0.2
          keys [
           <at 0 0.8 RCIM_Cubic>
           {arriveTangent=5.0121, arriveTangentWeight=0.0667, interpMode=RCIM_Cubic, leaveTangent=5.0121, tangentMode=RCTM_User, time=0.2, value=1}
           <at 0.2985 1 RCIM_Cubic>
           {arriveTangentWeight=0.1338, interpMode=RCIM_Cubic, tangentMode=RCTM_User, time=0.7, value=1}
           <at 1 1 RCIM_Cubic>
          ]
         }
         ZCurve {
          defaultValue=0.8
          keys [
           <at 0 0.1 RCIM_Cubic>
           {arriveTangent=5.0121, arriveTangentWeight=0.0667, interpMode=RCIM_Cubic, leaveTangent=5.0121, tangentMode=RCTM_User, time=0.2, value=1}
           <at 0.2986 1 RCIM_Cubic>
           {arriveTangentWeight=0.1338, interpMode=RCIM_Cubic, tangentMode=RCTM_User, time=0.7, value=1}
           <at 1 1 RCIM_Cubic>
          ]
         }
        }})
     )
     6=&Multiply_Vector(
      A=$Particles.Initial.Square Glow Scale
      B=&VectorFromCurve(VectorCurve=DataInterface() {propertyValues {
         XCurve {keys [<at 0 1>,<at 1 1>]}
         YCurve {
          defaultValue=0.8
          keys [
           <at 0 0 RCIM_Cubic>
           {arriveTangent=12.5697, arriveTangentWeight=0.0167, interpMode=RCIM_Cubic, leaveTangent=12.5697, tangentMode=RCTM_User, time=0.05, value=1}
           <at 0.15 1 RCIM_Cubic>
           {arriveTangent=-1.1765, interpMode=RCIM_Cubic, leaveTangent=-1.1765, time=0.7, value=1}
           <at 1 0 RCIM_Cubic>
          ]
         }
         ZCurve {
          defaultValue=0.1
          keys [
           <at 0 0 RCIM_Cubic>
           {arriveTangent=12.4571, arriveTangentWeight=0.0167, interpMode=RCIM_Cubic, leaveTangent=12.4571, tangentMode=RCTM_User, time=0.05, value=1}
           <at 0.15 1 RCIM_Cubic>
           {arriveTangent=-1.1765, interpMode=RCIM_Cubic, leaveTangent=-1.1765, time=0.7, value=1}
           <at 1 0 RCIM_Cubic>
          ]
         }
        }})
     )
     7=&Multiply_Vector(
      A=$Particles.Initial.Square Glow Scale
      B=&VectorFromCurve(VectorCurve=DataInterface() {propertyValues {
         XCurve {keys [<at 0 1>,<at 1 1>]}
         YCurve {
          defaultValue=0
          keys [
           <at 0 0 RCIM_Cubic>
           {arriveTangent=12.5697, arriveTangentWeight=0.0333, interpMode=RCIM_Cubic, leaveTangent=12.5697, tangentMode=RCTM_User, time=0.1, value=1}
           <at 0.2 1 RCIM_Cubic>
           {arriveTangent=-1.25, interpMode=RCIM_Cubic, leaveTangent=-1.25, time=0.7, value=1}
           <at 1 0 RCIM_Cubic>
          ]
         }
         ZCurve {
          defaultValue=0
          keys [
           <at 0 0 RCIM_Cubic>
           {arriveTangent=12.4571, arriveTangentWeight=0.0333, interpMode=RCIM_Cubic, leaveTangent=12.4571, tangentMode=RCTM_User, time=0.1, value=1}
           <at 0.2 1 RCIM_Cubic>
           {arriveTangent=-1.25, interpMode=RCIM_Cubic, leaveTangent=-1.25, time=0.7, value=1}
           <at 1 0 RCIM_Cubic>
          ]
         }
        }})
     )
     8=&Multiply_Vector(
      A=$Particles.Initial.Square Glow Scale
      B=&VectorFromCurve(VectorCurve=DataInterface() {propertyValues {
         XCurve {keys [<at 0 1>,<at 1 1>]}
         YCurve {
          defaultValue=0
          keys [
           <at 0 0 RCIM_Cubic>
           {arriveTangent=12.5697, arriveTangentWeight=0.0167, interpMode=RCIM_Cubic, leaveTangent=12.5697, tangentMode=RCTM_User, time=0.05, value=1}
           {arriveTangent=3.9351, arriveTangentWeight=0.0326, interpMode=RCIM_Cubic, leaveTangent=3.9351, tangentMode=RCTM_User, time=0.1477, value=1.05}
           {arriveTangentWeight=0.0393, interpMode=RCIM_Cubic, tangentMode=RCTM_User, time=0.2656, value=1.05}
           {arriveTangentWeight=0.2115, interpMode=RCIM_Cubic, tangentMode=RCTM_User, time=0.9, value=1.05}
           <at 1 0 RCIM_Cubic>
          ]
         }
         ZCurve {
          defaultValue=0
          keys [
           <at 0 0 RCIM_Cubic>
           {arriveTangent=12.4571, arriveTangentWeight=0.0167, interpMode=RCIM_Cubic, leaveTangent=12.4571, tangentMode=RCTM_User, time=0.05, value=1}
           {arriveTangent=5.1603, arriveTangentWeight=0.0336, interpMode=RCIM_Cubic, leaveTangent=5.1603, tangentMode=RCTM_User, time=0.1508, value=1.05}
           {arriveTangentWeight=0.039, interpMode=RCIM_Cubic, tangentMode=RCTM_User, time=0.2678, value=1.05}
           {arriveTangentWeight=0.2107, interpMode=RCIM_Cubic, tangentMode=RCTM_User, time=0.9, value=1.05}
           <at 1 0 RCIM_Cubic>
          ]
         }
        }})
     )
     9=&Multiply_Vector(
      A=$Particles.Initial.Square Glow Scale
      B=&VectorFromCurve(VectorCurve=DataInterface() {propertyValues {
         XCurve {keys [<at 0 1>,<at 1 1>]}
         YCurve {
          defaultValue=0
          keys [
           <at 0 0.8 RCIM_Cubic>
           {arriveTangent=12.5697, arriveTangentWeight=0.0071, interpMode=RCIM_Cubic, leaveTangent=12.5697, tangentMode=RCTM_User, time=0.0212, value=1}
           <at 0.0829 1 RCIM_Cubic>
           {arriveTangentWeight=0.059, interpMode=RCIM_Cubic, tangentMode=RCTM_User, time=0.26, value=1}
           <at 0.33 0 RCIM_Cubic>
           {arriveTangentWeight=0.0133, interpMode=RCIM_Cubic, tangentMode=RCTM_User, time=0.37, value=1}
           {arriveTangentWeight=0.0433, interpMode=RCIM_Cubic, tangentMode=RCTM_User, time=0.5, value=0.7}
           {arriveTangentWeight=0.1333, interpMode=RCIM_Cubic, tangentMode=RCTM_User, time=0.9, value=0.7}
           <at 1 0 RCIM_Cubic>
          ]
         }
         ZCurve {
          defaultValue=0
          keys [
           <at 0 0.8 RCIM_Cubic>
           {arriveTangent=12.4571, arriveTangentWeight=0.0071, interpMode=RCIM_Cubic, leaveTangent=12.4571, tangentMode=RCTM_User, time=0.0212, value=1}
           <at 0.0829 1 RCIM_Cubic>
           {arriveTangentWeight=0.059, interpMode=RCIM_Cubic, tangentMode=RCTM_User, time=0.26, value=1}
           <at 0.33 0 RCIM_Cubic>
           {arriveTangentWeight=0.0133, interpMode=RCIM_Cubic, tangentMode=RCTM_User, time=0.37, value=1}
           {arriveTangentWeight=0.0433, interpMode=RCIM_Cubic, tangentMode=RCTM_User, time=0.5, value=0.7}
           {arriveTangentWeight=0.1333, interpMode=RCIM_Cubic, tangentMode=RCTM_User, time=0.9, value=0.7}
           <at 1 0 RCIM_Cubic>
          ]
         }
        }})
     )
     Index=&SpawnGroup()
     Num=::NewEnumerator8
    )
   ) {isSetParameters=true}
  ]
  rendererClasses [&NiagaraSpriteRendererProperties, &NiagaraMeshRendererProperties]
  renderers [
   {rendererClass=&NiagaraSpriteRendererProperties}
   {rendererClass=&NiagaraMeshRendererProperties, rendererIndex=1}
  ]
 }
}
summary {
 systemName=NS_Emote_ScanEvade_SquaresPart1
 emitters [
  {
   emitterName=OuterFinalSquare
   rendererClasses [&NiagaraSpriteRendererProperties, &NiagaraMeshRendererProperties]
  }
  {emitterName=Spline_Position, rendererClasses [&NiagaraSpriteRendererProperties]}
  {emitterName=Splines, rendererClasses [&NiagaraMeshRendererProperties]}
  {emitterName=InnerCross, rendererClasses [&NiagaraMeshRendererProperties]}
  {emitterName=Squares, rendererClasses [&NiagaraSpriteRendererProperties, &NiagaraMeshRendererProperties]}
  {emitterName=Lights, rendererClasses [&NiagaraLightRendererProperties]}
 ]
 userVariables []
}
