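# Niagara system for part 2 of the squares VFX in the ScanEvade emote, showing a red, flickering, glowing square.
# The Square emitter burst-spawns fixed-orientation particles drawn by sprite, mesh and light renderers.
# Matching non-uniform curves indexed by System.Age animate sprite size and the custom Square Glow Scale through pop-in, dips and collapse.
# A separate alpha curve fades the light in and out, while the extra Light emitter has no renderers.
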NiagaraSystem /BRCosmetics/Animation/Game/MainPlayer/Emotes/ScanEvade/VFX/NS_Emote_ScanEvade_SquaresPart2
dependencies {
 usedDataInterfaces [
  &NiagaraDataInterfaceCurve
  &NiagaraDataInterfaceVector2DCurve
  &NiagaraDataInterfaceVectorCurve
 ]
 usedDynamicInputs []
 usedModules [&SystemState]
}
emitters {
 Light {rendererClasses [], renderers []}
 Square {
  emitterUpdateScript [
   &EmitterState()
   &SpawnBurst_Instantaneous(Use Spawn Probability=false, UseLoopCountLimit=false)
  ]
  particleSpawnScript [
   &InitializeParticle(Color=<rgba 1 0.0038 0 1>, Sprite Size=<20 20>, Sprite Size Mode=::NewEnumerator4, UsePositionOffset=false, Write Material Random=true)
   /SetVariables(
    Particles.Square Glow Scale:Vector3f=&Multiply_VectorByFloat(Vector=<1 1 1>, Float=0.34)
    Particles.Square Glow Size:Vector2f=&Vector2DFromFloat(Value=34.9175)
   ) {isSetParameters=true}
   &SpriteFacingAndAlignment(Sprite Alignment=<0 0 1>, Sprite Facing=<0 1 0>, WriteAlignment=true, WriteFacing=true)
  ]
  particleUpdateScript [
   &ParticleState()
   &ScaleSpriteSize(
    Non-Uniform Curve Index:NiagaraFloat=$System.Age
    Non-Uniform Curve Sprite Scale=DataInterface() {propertyValues {
      XCurve {keys [
        <at 0 0 RCIM_Cubic>
        {arriveTangentWeight=0.1273, interpMode=RCIM_Cubic, leaveTangentWeight=0.3441, tangentMode=RCTM_User, time=0.1663, value=1}
        {arriveTangentWeight=0.0586, interpMode=RCIM_Cubic, leaveTangentWeight=0.0435, tangentMode=RCTM_User, time=0.3033, value=1}
        {arriveTangentWeight=0.1183, interpMode=RCIM_Cubic, leaveTangentWeight=0.131, tangentMode=RCTM_User, time=0.4141, value=1.2}
        {
         arriveTangent=0.0022
         arriveTangentWeight=0.0842
         interpMode=RCIM_Cubic
         leaveTangent=4.3158
         leaveTangentWeight=0.096
         tangentMode=RCTM_Break
         time=0.5968
         value=1.2
        }
        {
         arriveTangent=-3.9205
         arriveTangentWeight=0.2931
         interpMode=RCIM_Cubic
         leaveTangent=-3.9205
         leaveTangentWeight=0.1167
         tangentMode=RCTM_User
         time=0.7782
         value=1.2
        }
        {arriveTangentWeight=0.2433, interpMode=RCIM_Cubic, tangentMode=RCTM_User, time=1, value=0}
       ]}
      YCurve {keys [
        <at 0 0 RCIM_Cubic>
        {arriveTangentWeight=0.1273, interpMode=RCIM_Cubic, leaveTangentWeight=0.3441, tangentMode=RCTM_User, time=0.1663, value=1}
        {arriveTangentWeight=0.2618, interpMode=RCIM_Cubic, leaveTangentWeight=0.0745, tangentMode=RCTM_User, time=0.2201, value=0}
        {arriveTangentWeight=0.0098, interpMode=RCIM_Cubic, leaveTangentWeight=0.0329, tangentMode=RCTM_User, time=0.2913, value=0}
        {arriveTangentWeight=0.0436, interpMode=RCIM_Cubic, leaveTangentWeight=0.0613, tangentMode=RCTM_User, time=0.4141, value=1.2}
        <at 0.541 0.25 RCIM_Cubic>
        {arriveTangentWeight=0.1297, interpMode=RCIM_Cubic, leaveTangentWeight=0.2777, tangentMode=RCTM_User, time=0.6447, value=1.2}
        {arriveTangentWeight=0.2433, interpMode=RCIM_Cubic, tangentMode=RCTM_User, time=0.7767, value=1.2}
        <at 0.9956 0 RCIM_Cubic>
       ]}
     }}
    Scale Sprite Size Mode=::NewEnumerator3
   )
   /SetVariables(
    Particles.Square Glow Color:LinearColor=&MakeLinearColorFromVectorAndFloat("Vector (RGB)"=&MakeVectorFromLinearColorRGB(Color=<rgba 1 0.0609 0.0573 1>), "Float (Alpha)"=0.8)
    Particles.Square Glow Scale:Vector3f=&Multiply_Vector(
     A=$Particles.Initial.Square Glow Scale
     B=&VectorFromCurve(
      VectorCurve=DataInterface() {propertyValues {
        XCurve {keys [<at 0 1>,<at 1 1>]}
        YCurve {
         defaultValue=0
         keys [
          <at 0 0 RCIM_Cubic>
          {arriveTangentWeight=0.0554, interpMode=RCIM_Cubic, tangentMode=RCTM_User, time=0.1663, value=1}
          {arriveTangentWeight=0.0457, interpMode=RCIM_Cubic, tangentMode=RCTM_User, time=0.3033, value=1}
          {arriveTangentWeight=0.0369, interpMode=RCIM_Cubic, tangentMode=RCTM_User, time=0.4141, value=1.2}
          {arriveTangent=0.0022, arriveTangentWeight=0.0609, interpMode=RCIM_Cubic, leaveTangent=4.3158, tangentMode=RCTM_Break, time=0.5968, value=1.2}
          {
           arriveTangent=-3.9205
           arriveTangentWeight=0.0604
           interpMode=RCIM_Cubic
           leaveTangent=-3.9205
           tangentMode=RCTM_User
           time=0.7782
           value=1.2
          }
          {arriveTangentWeight=0.0739, interpMode=RCIM_Cubic, tangentMode=RCTM_User, time=1, value=0}
         ]
        }
        ZCurve {
         defaultValue=0
         keys [
          <at 0 0 RCIM_Cubic>
          {arriveTangentWeight=0.0554, interpMode=RCIM_Cubic, tangentMode=RCTM_User, time=0.1663, value=1}
          {arriveTangentWeight=0.0179, interpMode=RCIM_Cubic, tangentMode=RCTM_User, time=0.2201, value=0}
          {arriveTangentWeight=0.0237, interpMode=RCIM_Cubic, tangentMode=RCTM_User, time=0.2913, value=0}
          {arriveTangentWeight=0.0409, interpMode=RCIM_Cubic, tangentMode=RCTM_User, time=0.4141, value=1.2}
          <at 0.541 0.25 RCIM_Cubic>
          {arriveTangentWeight=0.0346, interpMode=RCIM_Cubic, tangentMode=RCTM_User, time=0.6447, value=1.2}
          {arriveTangentWeight=0.044, interpMode=RCIM_Cubic, tangentMode=RCTM_User, time=0.7767, value=1.2}
          <at 0.9956 0 RCIM_Cubic>
         ]
        }
       }}
      CurveIndex=$System.Age
     )
    )
   ) {isSetParameters=true}
   &DynamicMaterialParameters(
    Index 0 Param 1=0.08
    Module.Index0Float0Write=true
    Module.Index0Float1Write=true
    Module.Index0Float2Write=true
    Module.Index0Float3Write=true
   )
   &Light_Attributes(
    Color=1
    Diffuse Scale=1
    Light Diffuse Scale=2
    Light Radius=50
    Light Specular Scale=2
    Radius=1
    Scale Alpha:NiagaraFloat=&FloatFromCurve(FloatCurve=DataInterface() {propertyValues {Curve {keys [
         <at 0 0 RCIM_Cubic>
         {arriveTangent=1.087, interpMode=RCIM_Cubic, leaveTangent=1.087, time=0.8, value=1}
         {arriveTangent=-5, interpMode=RCIM_Cubic, leaveTangent=-5, time=0.92, value=1}
         <at 1 0 RCIM_Cubic>
        ]}}})
    ScaleA=true
    ScaleRGB=true
    Specular Scale=1
   )
  ]
  rendererClasses [
   &NiagaraSpriteRendererProperties
   &NiagaraMeshRendererProperties
   &NiagaraLightRendererProperties
  ]
  renderers [
   {rendererClass=&NiagaraSpriteRendererProperties}
   {rendererClass=&NiagaraMeshRendererProperties, rendererIndex=1}
   {rendererClass=&NiagaraLightRendererProperties, rendererIndex=2}
  ]
 }
}
summary {
 systemName=NS_Emote_ScanEvade_SquaresPart2
 emitters [
  {
   emitterName=Square
   rendererClasses [
    &NiagaraSpriteRendererProperties
    &NiagaraMeshRendererProperties
    &NiagaraLightRendererProperties
   ]
  }
  {emitterName=Light, rendererClasses []}
 ]
 userVariables []
}
